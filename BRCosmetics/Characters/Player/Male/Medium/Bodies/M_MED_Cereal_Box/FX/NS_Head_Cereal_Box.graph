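# Niagara cosmetic effect that adds short-lived golden glints around the Cereal Box character's head.
# Glints maps spawn groups to bones or sockets on a user skeletal mesh, with group-based random offsets, while Glints001 samples the head bone with its own offset box.
# Both layer three sprite renderers, using curve-driven pulses and separate size, rotation and color attributes.
# User bool bisHeadHidden sets each particle's VisibilityTag.
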NiagaraSystem /BRCosmetics/Characters/Player/Male/Medium/Bodies/M_MED_Cereal_Box/FX/NS_Head_Cereal_Box
dependencies {usedDataInterfaces [&NiagaraDataInterfaceCurve], usedDynamicInputs [], usedModules [&SystemState]}
emitters {
 Glints {
  emitterUpdateScript [
   &EmitterState()
   &SpawnRate(SpawnRate=2.5, Use Spawn Probability=false)
  ]
  particleSpawnScript [
   &InitializeParticle(
    Color Maximum=<rgba 1 0.749 0.118 1>
    Color Minimum=<rgba 1 0.6663 0.2411 1>
    Color Mode=::NewEnumerator2
    Lifetime=&RandomRangeFloat(Minimum=0.6, Maximum=0.9)
    Sprite Size Mode=::NewEnumerator1
    Uniform Sprite Size Max=15
    Uniform Sprite Size Min=12
    UsePositionOffset=false
    Write Material Random=true
   )
   &FN_SetSpawnGroup(Number of Spawn Groups=17)
  ]
  particleUpdateScript [
   &ParticleState()
   &SkeletalMeshLocation(
    Bone / Socket Index=$Particles.SpawnGroup
    Bone / Socket Sampling Mode=::NewEnumerator5
    MaskSpawnGroup=false
    Mesh Sampling Type=::NewEnumerator6
    Sampled Position Offset=/Game/Effects/Niagara/DynamicInputs/Water/FN_PickVectorFromSpawnGroupCutoff(
     Under Cutoff Vector:Vector3f=&RandomRangeVector(Minimum=<-12 -17 -15>, Maximum=<12 0 10>)
     Over Cutoff Vector:Vector3f=&FN_PickVectorFromSpawnGroupCutoff(
      Under Cutoff Vector:Vector3f=&RandomRangeVector(Minimum=<-3 0 -3>, Maximum=<3 0 3>)
      Over Cutoff Vector:Vector3f=&RandomRangeVector(Minimum=<0 0 20>, Maximum=<0 0 0>)
      SpawnGroup:NiagaraInt32=$Particles.SpawnGroup
      Cutoff:NiagaraInt32=16
     )
     SpawnGroup:NiagaraInt32=$Particles.SpawnGroup
     Cutoff:NiagaraInt32=1
    )
    Skeletal Mesh=$User.New Skeletal Mesh
   )
   &DynamicMaterialParameters(
    Index 0 Param 1=1
    Index 0 Param 2=1
    Index 0 Param 3=1
    Module.Param0WriteEnabled=true
    Module.Param1WriteEnabled=true
    Module.Param2WriteEnabled=true
    Module.Param3WriteEnabled=true
   )
   &CameraOffset(Camera Offset Amount=8)
   &ScaleSpriteSize(Uniform Curve Sprite Scale=DataInterface() {propertyValues {Curve {keys [
        {arriveTangent=2.2037, interpMode=RCIM_Cubic, leaveTangent=2.2037, tangentMode=RCTM_User, time=0, value=0}
        {interpMode=RCIM_Cubic, tangentMode=RCTM_User, time=0.5, value=1}
        {tangentMode=RCTM_User, time=1, value=0}
       ]}}})
   /SetVariables(Particles.Renderer2Size:Vector2f=&Multiply_Vector2D(
     A=$Particles.Initial.SpriteSize
     B=&Vector2DFromFloat(Value=&Multiply_Float(
       A=0.6
       B=&FloatFromCurve(FloatCurve=DataInterface() {propertyValues {Curve {keys [
            {arriveTangent=2.2037, interpMode=RCIM_Cubic, leaveTangent=2.2037, tangentMode=RCTM_User, time=0.15, value=0}
            {interpMode=RCIM_Cubic, tangentMode=RCTM_User, time=0.5, value=1}
            {tangentMode=RCTM_User, time=0.85, value=0}
           ]}}})
      ))
    )) {isSetParameters=true, module=SetVariables_#1}
   /SetVariables(Particles.Renderer2_SpriteRotation:NiagaraFloat=&FloatFromCurve(FloatCurve=DataInterface() {propertyValues {Curve {keys [<at 0 1>,<at 1 0>]}}}, Scale Curve=-90)) {isSetParameters=true, module=SetVariables_#2}
   /SetVariables(Particles.Renderer3Color:LinearColor=<rgba 2 2 2 1>) {isSetParameters=true, module=SetVariables_#3}
   /SetVariables(Particles.Renderer3Size:Vector2f=&Multiply_Vector2D(A=$Particles.SpriteSize, B=&Vector2DFromFloat(Value=0.5))) {isSetParameters=true, module=SetVariables_#4}
   /SetVariables(Particles.VisibilityTag=&MakeCustomIntFromBool(Bool=$User.bisHeadHidden, True INT=1, False INT=0)) {isSetParameters=true, module=SetVariables_#5}
  ]
  rendererClasses [&NiagaraSpriteRendererProperties]
  renderers [
   {rendererClass=&NiagaraSpriteRendererProperties}
   {rendererClass=&NiagaraSpriteRendererProperties, rendererIndex=1}
   {rendererClass=&NiagaraSpriteRendererProperties, rendererIndex=2}
  ]
 }
 Glints001 {
  emitterUpdateScript [
   &EmitterState()
   &SpawnRate(SpawnRate=2.5, Use Spawn Probability=false)
  ]
  particleSpawnScript [
   &InitializeParticle(
    Color Maximum=<rgba 1 0.749 0.118 1>
    Color Minimum=<rgba 1 0.6663 0.2411 1>
    Color Mode=::NewEnumerator2
    Lifetime=&RandomRangeFloat(Minimum=0.6, Maximum=0.9)
    Sprite Size Mode=::NewEnumerator1
    Uniform Sprite Size Max=15
    Uniform Sprite Size Min=12
    UsePositionOffset=false
    Write Material Random=true
   )
   &FN_SetSpawnGroup(Number of Spawn Groups=17) {enabled=false}
  ]
  particleUpdateScript [
   &ParticleState()
   &SkeletalMeshLocation(
    Bone / Socket Sampling Mode=::NewEnumerator5
    MaskSpawnGroup=false
    Mesh Sampling Type=::NewEnumerator6
    Sampled Position Offset=&RandomRangeVector(Minimum=<-15 -20 -6>, Maximum=<15 -5 6>)
    Sampled Position Offset Coordinate Space:ENiagaraSkelMeshTransforms=::NewEnumerator2
    Skeletal Mesh=DataInterface() {propertyValues {
      DefaultMesh=null
      ExcludeBoneName=root
      OptionalComponentClass=null
      PreviewMesh=/BRCosmetics/Characters/Player/Male/Medium/Bodies/M_MED_Cereal_Box/Meshes/Parts/M_MED_Cereal_Box_FaceAcc
      SkinningMode=SkinOnTheFly
      SoftSourceActor=null
      SourceMode=Default
      UvSetIndex=0
      WholeMeshLOD=-1
      bExcludeBone=false
      bReadDeformedGeometry=true
      bRequireCurrentFrameData=true
      ComponentTags []
      FilteredBones [head]
      FilteredSockets []
      MeshUserParameter {parameter {name=None, typeDefHandle {registeredTypeIndex=1}, varData []}}
      SamplingRegions []
     }}
   )
   &DynamicMaterialParameters(
    Index 0 Param 1=1
    Index 0 Param 2=1
    Index 0 Param 3=1
    Module.Param0WriteEnabled=true
    Module.Param1WriteEnabled=true
    Module.Param2WriteEnabled=true
    Module.Param3WriteEnabled=true
   )
   &CameraOffset(Camera Offset Amount=8)
   &ScaleSpriteSize(Uniform Curve Sprite Scale=DataInterface() {propertyValues {Curve {keys [
        {arriveTangent=2.2037, interpMode=RCIM_Cubic, leaveTangent=2.2037, tangentMode=RCTM_User, time=0, value=0}
        {interpMode=RCIM_Cubic, tangentMode=RCTM_User, time=0.5, value=1}
        {tangentMode=RCTM_User, time=1, value=0}
       ]}}})
   /SetVariables(Particles.Renderer2Size:Vector2f=&Multiply_Vector2D(
     A=$Particles.Initial.SpriteSize
     B=&Vector2DFromFloat(Value=&Multiply_Float(
       A=0.6
       B=&FloatFromCurve(FloatCurve=DataInterface() {propertyValues {Curve {keys [
            {arriveTangent=2.2037, interpMode=RCIM_Cubic, leaveTangent=2.2037, tangentMode=RCTM_User, time=0.15, value=0}
            {interpMode=RCIM_Cubic, tangentMode=RCTM_User, time=0.5, value=1}
            {tangentMode=RCTM_User, time=0.85, value=0}
           ]}}})
      ))
    )) {isSetParameters=true, module=SetVariables_#1}
   /SetVariables(Particles.Renderer2_SpriteRotation:NiagaraFloat=&FloatFromCurve(FloatCurve=DataInterface() {propertyValues {Curve {keys [<at 0 1>,<at 1 0>]}}}, Scale Curve=-90)) {isSetParameters=true, module=SetVariables_#2}
   /SetVariables(Particles.Renderer3Color:LinearColor=<rgba 2 2 2 1>) {isSetParameters=true, module=SetVariables_#3}
   /SetVariables(Particles.Renderer3Size:Vector2f=&Multiply_Vector2D(A=$Particles.SpriteSize, B=&Vector2DFromFloat(Value=0.5))) {isSetParameters=true, module=SetVariables_#4}
   /SetVariables(Particles.VisibilityTag=&MakeCustomIntFromBool(Bool=$User.bisHeadHidden, True INT=1, False INT=0)) {isSetParameters=true, module=SetVariables_#5}
  ]
  rendererClasses [&NiagaraSpriteRendererProperties]
  renderers [
   {rendererClass=&NiagaraSpriteRendererProperties}
   {rendererClass=&NiagaraSpriteRendererProperties, rendererIndex=1}
   {rendererClass=&NiagaraSpriteRendererProperties, rendererIndex=2}
  ]
 }
}
summary {
 systemName=NS_Head_Cereal_Box
 emitters [
  {emitterName=Glints, rendererClasses [&NiagaraSpriteRendererProperties]}
  {emitterName=Glints001, rendererClasses [&NiagaraSpriteRendererProperties]}
 ]
 userVariables [
  User.bisHeadHidden {type=NiagaraBool}
  User.New Skeletal Mesh {
   type=NiagaraDataInterfaceSkeletalMesh
   defaultValue {
    struct=NiagaraExt_VariableValue_DataInterface
    value {dataInterface=&self:NiagaraDataInterfaceSkeletalMesh_0}
   }
  }
 ]
}
