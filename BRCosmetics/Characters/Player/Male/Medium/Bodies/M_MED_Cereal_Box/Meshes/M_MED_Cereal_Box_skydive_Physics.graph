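# Physics asset for the skydive mesh of the male medium Cereal Box character, giving it capsule bodies for physics.
# Capsules cover the pelvis, two spine bones, thighs and calves, plus a five-link dyn_tail chain under a small master body.
# The tail links limit twist much more tightly than swing, while the master-to-pelvis and upper-to-lower spine joints rotate freely.
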PhysicsAsset /BRCosmetics/Characters/Player/Male/Medium/Bodies/M_MED_Cereal_Box/Meshes/M_MED_Cereal_Box_skydive_Physics
bodies {
 calf_l [{shapeType=Capsule, length=36.3392, radius=5.069, center=<-20.6166 -1.7472 0.0192>, rotation=<rot -0.0589 -92.8216 -89.5509>}]
 calf_r [{shapeType=Capsule, length=38.4648, radius=5.0782, center=<20.8838 1.689 0.0171>, rotation=<rot -0.0599 -92.9009 -89.5428>}]
 dyn_tail_01 [{shapeType=Capsule, length=3.9628, radius=4.3578, shapeName=dyn_tail_01_capsule, center=<5 0 0>, rotation=<rot 0 0 90>}]
 dyn_tail_02 [{shapeType=Capsule, length=9.5114, radius=7.7166, center=<5.5 0 0>, rotation=<rot 0 0 90>}]
 dyn_tail_03 [{shapeType=Capsule, length=2.4761, radius=9.5295, center=<7 0 0>, rotation=<rot 0 0 90>}]
 dyn_tail_04 [{shapeType=Capsule, length=7.0752, radius=5.2007, shapeName=dyn_tail_04_capsule, center=<7 0 0>, rotation=<rot 0 0 90>}]
 dyn_tail_05 [{shapeType=Capsule, length=6.2459, radius=4.3745, shapeName=dyn_tail_05_capsule, center=<7.5 0 0>, rotation=<rot 90 0 0>}]
 dyn_tail_MSTR [{shapeType=Capsule, length=0.505, radius=0.505, shapeName=dyn_tail_MSTR_capsule, rotation=<rot 0 -90 0>}]
 pelvis [{shapeType=Capsule, length=9.0958, radius=10.0069, center=<-6.7548 0.9327 0.0001>}]
 spine_01 [{shapeType=Capsule, length=9.4482, radius=7.6231, center=<-4.1667 0.7083 0>}]
 spine_04 [{shapeType=Capsule, length=20.5881, radius=15.2297, shapeName=spine_04_capsule, center=<4.36 -17.1687 0.0618>, rotation=<rot 90 0 0>}]
 thigh_l [{shapeType=Capsule, length=25.0039, radius=7.5234, center=<-20.5434 0.5658 0.7282>, rotation=<rot 0.4324 89.5526 -85.8198>}]
 thigh_r [{shapeType=Capsule, length=23.8229, radius=7.6282, center=<20.9429 -0.6031 -0.8746>, rotation=<rot -0.0271 92.2581 -90.4972>}]
}
constraints [
 {bone1Name=spine_01, bone2Name=pelvis, swing1LimitDegrees=45, swing2LimitDegrees=45, twistLimitDegrees=45}
 {bone1Name=thigh_l, bone2Name=pelvis, swing1LimitDegrees=45, swing2LimitDegrees=45, twistLimitDegrees=45}
 {bone1Name=calf_l, bone2Name=thigh_l, swing1LimitDegrees=45, swing2LimitDegrees=45, twistLimitDegrees=45}
 {bone1Name=thigh_r, bone2Name=pelvis, swing1LimitDegrees=45, swing2LimitDegrees=45, twistLimitDegrees=45}
 {bone1Name=calf_r, bone2Name=thigh_r, swing1LimitDegrees=45, swing2LimitDegrees=45, twistLimitDegrees=45}
 {bone1Name=dyn_tail_02, bone2Name=dyn_tail_01, swing1LimitDegrees=45, swing2LimitDegrees=45, twistLimitDegrees=5}
 {bone1Name=dyn_tail_03, bone2Name=dyn_tail_02, swing1LimitDegrees=45, swing2LimitDegrees=45, twistLimitDegrees=5}
 {bone1Name=dyn_tail_05, bone2Name=dyn_tail_04, swing1LimitDegrees=45, swing2LimitDegrees=45, twistLimitDegrees=5}
 {bone1Name=dyn_tail_04, bone2Name=dyn_tail_03, swing1LimitDegrees=45, swing2LimitDegrees=45, twistLimitDegrees=5}
 {bone1Name=dyn_tail_01, bone2Name=dyn_tail_MSTR, swing1LimitDegrees=45, swing2LimitDegrees=45, twistLimitDegrees=5}
 {
  bone1Name=dyn_tail_MSTR
  bone2Name=pelvis
  swing1LimitDegrees=45
  swing1Motion=Free
  swing2LimitDegrees=45
  swing2Motion=Free
  twistLimitDegrees=45
  twistMotion=Free
 }
 {
  bone1Name=spine_04
  bone2Name=spine_01
  swing1LimitDegrees=45
  swing1Motion=Free
  swing2LimitDegrees=45
  swing2Motion=Free
  twistLimitDegrees=45
  twistMotion=Free
 }
]
