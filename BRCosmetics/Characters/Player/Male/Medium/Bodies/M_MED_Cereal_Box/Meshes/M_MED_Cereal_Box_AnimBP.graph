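# Anim Blueprint for the male Cereal Box cosmetic body, adding secondary motion to tail and jacket bones.
# bRigidBodyEnabled picks RigidBody simulation, with a separate skydive physics asset, or a tail chain of Look At, AnimDynamics and Trail controller.
# Position constraints and AnimDynamics then drive extra tail bones, followed by hand and forearm scale tweaks.
# Jacket AnimDynamics chains use spherical limits on spine and pelvis, some weighted by backpack or jacket variables.
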AnimBlueprint /BRCosmetics/Characters/Player/Male/Medium/Bodies/M_MED_Cereal_Box/Meshes/M_MED_Cereal_Box_AnimBP
target_skeleton=/Game/Characters/Player/Male/Medium/Bodies/M_MED_Cereal_Box/Meshes/M_MED_Cereal_Box_Skeleton
graphs [
 AnimGraph {nodes [
   n1 = RigidBody!(
    ComponentPose=Local To Component(LocalPose=Use cached pose 'prerigid'())
    Alpha
    OverrideWorldGravity:Vector=Get GravityOverride()
    ExternalForce
    ComponentLinearAccScale:Vector=Get AccScale() -> (AccScale:Vector)
    ComponentLinearVelScale:Vector=Get VelScale() -> (VelScale:Vector)
   ) -> (Pose) {
    bClampLinearTranslationLimitToRefPose=true
    bOverrideWorldGravity=true
    overridePhysicsAsset=&/M_MED_Cereal_Box_skydive_Physics
    simulationSpace=BaseBoneSpace
    alphaBoolBlend {blendOption=Linear}
    baseBoneRef {boneName=dyn_root_simspace}
    componentAppliedLinearAccClamp=<1500 1500 1500>
    componentLinearAccScale=<0.25 0.25 0.4>
    componentLinearVelScale=<0.2 0.2 0.2>
   }
   n2 = RigidBody(
    ComponentPose=Local To Component(LocalPose=Use cached pose 'prerigid'())
    OverrideWorldGravity=Get GravityOverride()
    ComponentLinearAccScale="0.25,0.25,0.4"
    ComponentLinearVelScale="0.2,0.2,0.2"
   ) {
    bClampLinearTranslationLimitToRefPose=true
    bOverrideWorldGravity=true
    overridePhysicsAsset=/Game/Characters/Player/Male/Medium/Bodies/M_MED_Cereal_Box/Meshes/M_MED_Cereal_Box_Physics
    simulationSpace=BaseBoneSpace
    alphaBoolBlend {blendOption=Linear}
    baseBoneRef {boneName=dyn_root_simspace}
    componentAppliedLinearAccClamp=<3000 3000 3000>
    componentLinearAccScale=<0.25 0.25 0.4>
    componentLinearVelScale=<0.2 0.2 0.2>
   }
   n3 = Blend Poses by bool(BlendPose_0=Component To Local(ComponentPose=@n1), BlendPose_1=Component To Local(ComponentPose=@n2), bActiveValue=Get bIsSkyDiving()) {blendType=Linear}
   n4 = Look At - Bone: dyn_tail_02!(ComponentPose=Local To Component(LocalPose=Use cached pose 'prerigid'()), bAlphaBoolEnabled:bool=Get Skydive_off() -> (Skydive_off:bool)) -> (Pose) {
    alphaInputType=Bool
    alphaBoolBlend {blendInTime=0.3, blendOption=Linear, blendOutTime=0.3}
    boneToModify {boneName=dyn_tail_02}
    lookAtLocation=<0 -30>
    lookAtTarget {boneReference {boneName=dyn_root_simspace}}
    lookAt_Axis {axis=<1 0>}
   }
   n5 = Anim Dynamics - Bone: dyn_tail_01(ComponentPose=@n4, Alpha=1, ExternalForce="0,0,0") {
    angularDampingOverride=0.9999
    angularSpringConstant=25
    bAngularSpring=true
    bGravityOverrideInSimSpace=true
    bOverrideAngularDamping=true
    simulationSpace=BoneRelative
    alphaBoolBlend {blendOption=Linear}
    boundBone {boneName=dyn_tail_01}
    componentAppliedLinearAccClamp=<3500 3500 3500>
    componentLinearAccScale=<1.2 1.2 1.2>
    componentLinearVelScale=<1.2 1.2 1.2>
    physicsBodyDefinitions [{
      sphereCollisionRadius=0
      boundBone {boneName=dyn_tail_01}
      boxExtents=<20 20 10>
      constraintSetup {angularLimitsMax=<1 30 15>, angularLimitsMin=<-1 -30 -15>, angularTarget=<0 0 0>}
      localJointOffset=<2 0 0>
     }]
    relativeSpaceBone {boneName=dyn_root_simspace}
   }
   n6 = Anim Dynamics - Bone: dyn_tail_02(ComponentPose=@n5, Alpha=1, ExternalForce="0,0,0") {
    angularDampingOverride=0.9999
    angularSpringConstant=25
    bAngularSpring=true
    bGravityOverrideInSimSpace=true
    bOverrideAngularDamping=true
    simulationSpace=BoneRelative
    alphaBoolBlend {blendOption=Linear}
    boundBone {boneName=dyn_tail_02}
    componentAppliedLinearAccClamp=<3500 3500 3500>
    componentLinearAccScale=<1.2 1.2 1.2>
    componentLinearVelScale=<1.2 1.2 1.2>
    physicsBodyDefinitions [{
      sphereCollisionRadius=0
      boundBone {boneName=dyn_tail_02}
      boxExtents=<20 20 10>
      constraintSetup {angularLimitsMax=<1 30 15>, angularLimitsMin=<-1 -30 -15>, angularTarget=<0 0 0>}
      localJointOffset=<2 0 0>
     }]
    relativeSpaceBone {boneName=dyn_root_simspace}
   }
   n7 = Anim Dynamics - Bone: dyn_tail_03(ComponentPose=@n6, Alpha=1, ExternalForce="0,0,0") {
    angularDampingOverride=0.9999
    angularSpringConstant=25
    bAngularSpring=true
    bGravityOverrideInSimSpace=true
    bOverrideAngularDamping=true
    simulationSpace=BoneRelative
    alphaBoolBlend {blendOption=Linear}
    boundBone {boneName=dyn_tail_03}
    componentAppliedLinearAccClamp=<3500 3500 3500>
    componentLinearAccScale=<1.2 1.2 1.2>
    componentLinearVelScale=<1.2 1.2 1.2>
    physicsBodyDefinitions [{
      sphereCollisionRadius=0
      boundBone {boneName=dyn_tail_03}
      boxExtents=<20 20 10>
      constraintSetup {angularLimitsMax=<1 30 15>, angularLimitsMin=<-1 -30 -15>, angularTarget=<0 0 0>}
      localJointOffset=<2 0 0>
     }]
    relativeSpaceBone {boneName=dyn_root_simspace}
   }
   n8 = Anim Dynamics - Bone: dyn_tail_04(ComponentPose=@n7, Alpha=1, ExternalForce="0,0,0") {
    angularDampingOverride=0.9999
    angularSpringConstant=25
    bAngularSpring=true
    bGravityOverrideInSimSpace=true
    bOverrideAngularDamping=true
    simulationSpace=BoneRelative
    alphaBoolBlend {blendOption=Linear}
    boundBone {boneName=dyn_tail_04}
    componentAppliedLinearAccClamp=<3500 3500 3500>
    componentLinearAccScale=<1.2 1.2 1.2>
    componentLinearVelScale=<1.2 1.2 1.2>
    physicsBodyDefinitions [{
      sphereCollisionRadius=0
      boundBone {boneName=dyn_tail_04}
      boxExtents=<20 20 10>
      constraintSetup {angularLimitsMax=<1 30 15>, angularLimitsMin=<-1 -30 -15>, angularTarget=<0 0 0>}
      localJointOffset=<2 0 0>
     }]
    relativeSpaceBone {boneName=dyn_root_simspace}
   }
   n9 = Anim Dynamics - Bone: dyn_tail_05(ComponentPose=@n8, Alpha=1, ExternalForce="0,0,0") {
    angularDampingOverride=0.9999
    angularSpringConstant=25
    bAngularSpring=true
    bGravityOverrideInSimSpace=true
    bOverrideAngularDamping=true
    simulationSpace=BoneRelative
    alphaBoolBlend {blendOption=Linear}
    boundBone {boneName=dyn_tail_05}
    componentAppliedLinearAccClamp=<3500 3500 3500>
    componentLinearAccScale=<1.2 1.2 1.2>
    componentLinearVelScale=<1.2 1.2 1.2>
    physicsBodyDefinitions [{
      sphereCollisionRadius=0
      boundBone {boneName=dyn_tail_05}
      boxExtents=<20 20 10>
      constraintSetup {angularLimitsMax=<1 30 15>, angularLimitsMin=<-1 -30 -15>, angularTarget=<0 0 0>}
      localJointOffset=<2 0 0>
     }]
    relativeSpaceBone {boneName=dyn_root_simspace}
   }
   n10 = Trail controller - Bone: dyn_tail_05!(ComponentPose=@n9, Alpha=Get BwdMovement() -> (BwdMovement:double), RelaxationSpeedScale:float=1, FakeVelocity:Vector="0,0,0") -> (Pose) {
    bUsePlanarLimit=true
    chainLength=4
    debugLifeTime=0.1
    maxDeltaTime=0.0333
    relaxationSpeedScale=1.2
    alphaScaleBiasClamp {bInterpResult=true, interpSpeedDecreasing=20, interpSpeedIncreasing=5}
    rotationLimits [{limitMax=<180 180 180>}, {limitMax=<180 180 180>}, {limitMax=<180 180 180>}, {limitMax=<180 180 180>}]
    rotationOffsets [<0 0 0>, <0 0 0>, <0 0 0>, <0 0 0>]
    trailBone {boneName=dyn_tail_05}
    trailRelaxationSpeed {editorCurveData {keys [<at 0 20 RCIM_Cubic>, {arriveTangent=-27.7668, interpMode=RCIM_Cubic, leaveTangent=-27.7668, tangentMode=RCTM_User, time=1, value=5}]}}
   }
   n11 = "Transform (Modify) Bone - Bone: dyn_jacket_e_MSTR"!(ComponentPose=@n10, Alpha=1, Translation="0,0,5") -> (Pose) {
    rotationSpace=BCS_WorldSpace
    scaleMode=BMM_Additive
    scaleSpace=BCS_BoneSpace
    translationMode=BMM_Additive
    translationSpace=BCS_BoneSpace
    alphaBoolBlend {blendInTime=0.1, blendOption=Linear, blendOutTime=0.1}
    alphaScaleBiasClamp {bInterpResult=true}
    boneToModify {boneName=dyn_jacket_e_MSTR}
    translation {z=5}
   }
   n12 = Component To Local(ComponentPose=@n11)
   n13 = Get bRigidBodyEnabled()
   n14 = "Constraint - dyn_tail_extra_a (P)"(
    ComponentPose=Local To Component(LocalPose=Blend Poses by bool(BlendPose_0=@n3, BlendPose_1=@n12, bActiveValue=@n13) {blendType=Linear})
    Alpha=1
    ConstraintWeights_0=1
   ) {
    alphaBoolBlend {blendOption=Linear}
    boneToModify {boneName=dyn_tail_extra_a}
    constraintSetup [{targetBone {boneName=dyn_tail_02}}]
    constraintWeights [1]
   }
   n15 = "Constraint - dyn_tail_extra_b (PP)"!(ComponentPose=@n14, Alpha=1, ConstraintWeights_0=0.2367, ConstraintWeights_1:float=0.7633) -> (Pose) {
    alphaBoolBlend {blendOption=Linear}
    boneToModify {boneName=dyn_tail_extra_b}
    constraintSetup [{targetBone {boneName=dyn_tail_02}}, {targetBone {boneName=dyn_tail_03}}]
    constraintWeights [1,1]
   }
   n16 = "Constraint - dyn_tail_extra_c (PP)"!(ComponentPose=@n15, Alpha=1, ConstraintWeights_0=0.0469, ConstraintWeights_1:float=0.9531) -> (Pose) {
    alphaBoolBlend {blendOption=Linear}
    boneToModify {boneName=dyn_tail_extra_c}
    constraintSetup [{targetBone {boneName=dyn_tail_02}}, {targetBone {boneName=dyn_tail_03}}]
    constraintWeights [1,1]
   }
   n17 = "Constraint - dyn_tail_extra_d (PPP)"!(ComponentPose=@n16, Alpha=1, ConstraintWeights_0=0.0064, ConstraintWeights_1:float=0.9903, ConstraintWeights_2:float=0.0033) -> (Pose) {
    alphaBoolBlend {blendOption=Linear}
    boneToModify {boneName=dyn_tail_extra_d}
    constraintSetup [{targetBone {boneName=dyn_tail_03}}, {targetBone {boneName=dyn_tail_04}}, {targetBone {boneName=dyn_tail_05}}]
    constraintWeights [1,1,1]
   }
   n18 = Anim Dynamics - Bone: dyn_tail_extra_a(ComponentPose=@n17, Alpha=1, ExternalForce="0,0,0") {
    angularDampingOverride=0.9999
    angularSpringConstant=25
    bAngularSpring=true
    bGravityOverrideInSimSpace=true
    bOverrideAngularDamping=true
    simulationSpace=BoneRelative
    alphaBoolBlend {blendOption=Linear}
    boundBone {boneName=dyn_tail_extra_a}
    componentAppliedLinearAccClamp=<3500 3500 3500>
    componentLinearAccScale=<1.2 1.2 1.2>
    componentLinearVelScale=<1.2 1.2 1.2>
    physicsBodyDefinitions [{
      sphereCollisionRadius=0
      boundBone {boneName=dyn_tail_extra_a}
      boxExtents=<20 20 10>
      constraintSetup {angularLimitsMax=<1 30 15>, angularLimitsMin=<-1 -30 -15>, angularTarget=<0 0 0>}
      localJointOffset=<2 0 0>
     }]
    relativeSpaceBone {boneName=dyn_root_simspace}
   }
   n19 = Anim Dynamics - Bone: dyn_tail_extra_b(ComponentPose=@n18, Alpha=1, ExternalForce="0,0,0") {
    angularDampingOverride=0.9999
    angularSpringConstant=25
    bAngularSpring=true
    bGravityOverrideInSimSpace=true
    bOverrideAngularDamping=true
    simulationSpace=BoneRelative
    alphaBoolBlend {blendOption=Linear}
    boundBone {boneName=dyn_tail_extra_b}
    componentAppliedLinearAccClamp=<3500 3500 3500>
    componentLinearAccScale=<1.2 1.2 1.2>
    componentLinearVelScale=<1.2 1.2 1.2>
    physicsBodyDefinitions [{
      sphereCollisionRadius=0
      boundBone {boneName=dyn_tail_extra_b}
      boxExtents=<20 20 10>
      constraintSetup {angularLimitsMax=<1 30 15>, angularLimitsMin=<-1 -30 -15>, angularTarget=<0 0 0>}
      localJointOffset=<2 0 0>
     }]
    relativeSpaceBone {boneName=dyn_root_simspace}
   }
   n20 = Anim Dynamics - Bone: dyn_tail_extra_c(ComponentPose=@n19, Alpha=1, ExternalForce="0,0,0") {
    angularDampingOverride=0.9999
    angularSpringConstant=25
    bAngularSpring=true
    bGravityOverrideInSimSpace=true
    bOverrideAngularDamping=true
    simulationSpace=BoneRelative
    alphaBoolBlend {blendOption=Linear}
    boundBone {boneName=dyn_tail_extra_c}
    componentAppliedLinearAccClamp=<3500 3500 3500>
    componentLinearAccScale=<1.2 1.2 1.2>
    componentLinearVelScale=<1.2 1.2 1.2>
    physicsBodyDefinitions [{
      sphereCollisionRadius=0
      boundBone {boneName=dyn_tail_extra_c}
      boxExtents=<20 20 10>
      constraintSetup {angularLimitsMax=<1 30 15>, angularLimitsMin=<-1 -30 -15>, angularTarget=<0 0 0>}
      localJointOffset=<2 0 0>
     }]
    relativeSpaceBone {boneName=dyn_root_simspace}
   }
   n21 = Anim Dynamics - Bone: dyn_tail_extra_d(ComponentPose=@n20, Alpha=1, ExternalForce="0,0,0") {
    angularDampingOverride=0.9999
    angularSpringConstant=25
    bAngularSpring=true
    bGravityOverrideInSimSpace=true
    bOverrideAngularDamping=true
    simulationSpace=BoneRelative
    alphaBoolBlend {blendOption=Linear}
    boundBone {boneName=dyn_tail_extra_d}
    componentAppliedLinearAccClamp=<3500 3500 3500>
    componentLinearAccScale=<1.2 1.2 1.2>
    componentLinearVelScale=<1.2 1.2 1.2>
    physicsBodyDefinitions [{
      sphereCollisionRadius=0
      boundBone {boneName=dyn_tail_extra_d}
      boxExtents=<20 20 10>
      constraintSetup {angularLimitsMax=<1 30 15>, angularLimitsMin=<-1 -30 -15>, angularTarget=<0 0 0>}
      localJointOffset=<2 0 0>
     }]
    relativeSpaceBone {boneName=dyn_root_simspace}
   }
   n22 = "Transform (Modify) Bone - Bone: hand_l"!(ComponentPose=@n21, Alpha=1, Translation="0,0,0", Scale="0.2,0.2,0.2") -> (Pose) {
    rotationSpace=BCS_WorldSpace
    scaleMode=BMM_Additive
    scaleSpace=BCS_BoneSpace
    translationMode=BMM_Additive
    translationSpace=BCS_BoneSpace
    alphaBoolBlend {blendInTime=0.1, blendOption=Linear, blendOutTime=0.1}
    boneToModify {boneName=hand_l}
    scale=<0.2 0.2 0.2>
   }
   n23 = "Transform (Modify) Bone - Bone: dyn_hand_r"!(ComponentPose=@n22, Alpha=1, Translation="0,0,0", Scale="0.2,0.2,0.2") -> (Pose) {
    rotationSpace=BCS_WorldSpace
    scaleMode=BMM_Additive
    scaleSpace=BCS_BoneSpace
    translationMode=BMM_Additive
    translationSpace=BCS_BoneSpace
    alphaBoolBlend {blendInTime=0.1, blendOption=Linear, blendOutTime=0.1}
    boneToModify {boneName=dyn_hand_r}
    scale=<0.2 0.2 0.2>
   }
   n24 = "Transform (Modify) Bone - Bone: dyn_hoof_hand_l"!(ComponentPose=@n23, Alpha=1, Translation="5,0,0", Scale="1.2,1.2,1.2") -> (Pose) {
    rotationSpace=BCS_WorldSpace
    scaleMode=BMM_Additive
    scaleSpace=BCS_BoneSpace
    translationMode=BMM_Additive
    translationSpace=BCS_BoneSpace
    alphaBoolBlend {blendInTime=0.1, blendOption=Linear, blendOutTime=0.1}
    boneToModify {boneName=dyn_hoof_hand_l}
    scale=<1.2 1.2 1.2>
    translation {x=5}
   }
   n25 = "Transform (Modify) Bone - Bone: dyn_hoof_hand_r"!(ComponentPose=@n24, Alpha=1, Translation="-5,0,0", Scale="1.2,1.2,1.2") -> (Pose) {
    rotationSpace=BCS_WorldSpace
    scaleMode=BMM_Additive
    scaleSpace=BCS_BoneSpace
    translationMode=BMM_Additive
    translationSpace=BCS_BoneSpace
    alphaBoolBlend {blendInTime=0.1, blendOption=Linear, blendOutTime=0.1}
    boneToModify {boneName=dyn_hoof_hand_r}
    scale=<1.2 1.2 1.2>
    translation {x=-5}
   }
   n26 = "Transform (Modify) Bone - Bone: dyn_hoof_hand_l"(ComponentPose=@n25, Alpha=1, Translation="0,0,0", Scale="0.1,0.1,0.1") {
    rotationSpace=BCS_WorldSpace
    scaleMode=BMM_Additive
    scaleSpace=BCS_BoneSpace
    translationMode=BMM_Additive
    translationSpace=BCS_BoneSpace
    alphaBoolBlend {blendInTime=0.1, blendOption=Linear, blendOutTime=0.1}
    boneToModify {boneName=dyn_hoof_hand_l}
    scale=<0.1 0.1 0.1>
   }
   n27 = "Transform (Modify) Bone - Bone: dyn_hoof_hand_r"(ComponentPose=@n26, Alpha=1, Translation="0,0,0", Scale="0.1,0.1,0.1") {
    rotationSpace=BCS_WorldSpace
    scaleMode=BMM_Additive
    scaleSpace=BCS_BoneSpace
    translationMode=BMM_Additive
    translationSpace=BCS_BoneSpace
    alphaBoolBlend {blendInTime=0.1, blendOption=Linear, blendOutTime=0.1}
    boneToModify {boneName=dyn_hoof_hand_r}
    scale=<0.1 0.1 0.1>
   }
   n28 = "Transform (Modify) Bone - Bone: lowerarm_twist_01_l"!(ComponentPose=@n27, Alpha=1, Translation="0,0,0", Scale="1,0.1,0.1") -> (Pose) {
    rotationSpace=BCS_WorldSpace
    scaleMode=BMM_Additive
    scaleSpace=BCS_BoneSpace
    translationMode=BMM_Additive
    translationSpace=BCS_BoneSpace
    alphaBoolBlend {blendInTime=0.1, blendOption=Linear, blendOutTime=0.1}
    boneToModify {boneName=lowerarm_twist_01_l}
    scale {y=0.1, z=0.1}
   }
   n29 = "Transform (Modify) Bone - Bone: lowerarm_twist_01_r"!(ComponentPose=@n28, Alpha=1, Translation="0,0,0", Scale="1,0.1,0.1") -> (Pose) {
    rotationSpace=BCS_WorldSpace
    scaleMode=BMM_Additive
    scaleSpace=BCS_BoneSpace
    translationMode=BMM_Additive
    translationSpace=BCS_BoneSpace
    alphaBoolBlend {blendInTime=0.1, blendOption=Linear, blendOutTime=0.1}
    boneToModify {boneName=lowerarm_twist_01_r}
    scale {y=0.1, z=0.1}
   }
   Output Pose(Result=Component To Local(ComponentPose=@n29)) {layerGroup=None}
   n30 = Get jacket dynamics() -> (jacket dynamics:double)
   n31 = Get GravityOverride_2() -> (GravityOverride_2:Vector)
   n32 = Anim Dynamics - Bone: dyn_jacket_a_01!(ComponentPose=Local To Component(LocalPose=Use cached pose 'prefrontend'()), Alpha=@n30, GravityOverride:Vector=@n31, ExternalForce="0,0,0") -> (Pose) {
    angularDampingOverride=0.9999
    angularSpringConstant=50
    bAngularSpring=true
    bGravityOverrideInSimSpace=true
    bOverrideAngularDamping=true
    bUseGravityOverride=true
    simulationSpace=BoneRelative
    alphaBoolBlend {blendOption=Linear}
    boundBone {boneName=dyn_jacket_a_01}
    componentAppliedLinearAccClamp=<3500 3500 3500>
    componentLinearAccScale=<1 1 1>
    componentLinearVelScale=<1 1 1>
    physicsBodyDefinitions [{
      sphereCollisionRadius=0
      boundBone {boneName=dyn_jacket_a_01}
      boxExtents=<10 20 20>
      constraintSetup {angularLimitsMax=<1 0 25>, angularLimitsMin=<-1 -25 -25>, angularTarget=<0 0 0>}
      localJointOffset=<2 0 0>
     }]
    relativeSpaceBone {boneName=dyn_root_simspace}
   }
   n33 = Anim Dynamics - Bone: dyn_jacket_a_02!(ComponentPose=@n32, Alpha=@n30, GravityOverride:Vector=@n31, ExternalForce="0,0,0") -> (Pose) {
    angularDampingOverride=0.9999
    angularSpringConstant=50
    bAngularSpring=true
    bGravityOverrideInSimSpace=true
    bOverrideAngularDamping=true
    bUseGravityOverride=true
    simulationSpace=BoneRelative
    alphaBoolBlend {blendOption=Linear}
    boundBone {boneName=dyn_jacket_a_02}
    componentAppliedLinearAccClamp=<3500 3500 3500>
    componentLinearAccScale=<1 1 1>
    componentLinearVelScale=<1 1 1>
    physicsBodyDefinitions [{
      sphereCollisionRadius=0
      boundBone {boneName=dyn_jacket_a_02}
      boxExtents=<10 20 20>
      constraintSetup {angularLimitsMax=<1 0 25>, angularLimitsMin=<-1 -25 -25>, angularTarget=<0 0 0>}
      localJointOffset=<2 0 0>
     }]
    relativeSpaceBone {boneName=dyn_root_simspace}
   }
   n34 = Anim Dynamics - Bone: dyn_jacket_a_03!(ComponentPose=@n33, Alpha=@n30, GravityOverride:Vector=@n31, ExternalForce="0,0,0") -> (Pose) {
    angularDampingOverride=0.9999
    angularSpringConstant=50
    bAngularSpring=true
    bGravityOverrideInSimSpace=true
    bOverrideAngularDamping=true
    bUseGravityOverride=true
    bUseSphericalLimits=true
    simulationSpace=BoneRelative
    alphaBoolBlend {blendOption=Linear}
    boundBone {boneName=dyn_jacket_a_03}
    componentAppliedLinearAccClamp=<3500 3500 3500>
    componentLinearAccScale=<1 1 1>
    componentLinearVelScale=<1 1 1>
    physicsBodyDefinitions [{
      collisionType=InnerSphere
      sphereCollisionRadius=1
      boundBone {boneName=dyn_jacket_a_03}
      boxExtents=<10 20 20>
      constraintSetup {angularLimitsMax=<1 0 25>, angularLimitsMin=<-1 -25 -25>, angularTarget=<0 0 0>}
      localJointOffset=<2 0 0>
     }]
    relativeSpaceBone {boneName=dyn_root_simspace}
    sphericalLimits [
     {limitRadius=8.5, drivingBone {boneName=spine_01}, sphereLocalOffset=<0 0 0>}
     {limitRadius=10, drivingBone {boneName=pelvis}, sphereLocalOffset=<0 0 0>}
    ]
   }
   n35 = Anim Dynamics - Bone: dyn_jacket_b_01!(ComponentPose=@n34, Alpha=1, GravityOverride:Vector=@n31, ExternalForce="0,0,0") -> (Pose) {
    angularDampingOverride=0.9999
    angularSpringConstant=50
    bAngularSpring=true
    bGravityOverrideInSimSpace=true
    bOverrideAngularDamping=true
    bUseGravityOverride=true
    simulationSpace=BoneRelative
    alphaBoolBlend {blendOption=Linear}
    boundBone {boneName=dyn_jacket_b_01}
    componentAppliedLinearAccClamp=<3500 3500 3500>
    componentLinearAccScale=<0.8 0.8 0.8>
    componentLinearVelScale=<0.8 0.8 0.8>
    physicsBodyDefinitions [{
      sphereCollisionRadius=0
      boundBone {boneName=dyn_jacket_b_01}
      boxExtents=<10 20 20>
      constraintSetup {angularLimitsMax=<1 0 15>, angularLimitsMin=<-1 -30 -15>, angularTarget=<0 0 0>}
      localJointOffset=<2 0 0>
     }]
    relativeSpaceBone {boneName=dyn_root_simspace}
   }
   n36 = Anim Dynamics - Bone: dyn_jacket_b_02!(ComponentPose=@n35, Alpha=1, GravityOverride:Vector=@n31, ExternalForce="0,0,0") -> (Pose) {
    angularDampingOverride=0.9999
    angularSpringConstant=50
    bAngularSpring=true
    bGravityOverrideInSimSpace=true
    bOverrideAngularDamping=true
    bUseGravityOverride=true
    bUseSphericalLimits=true
    simulationSpace=BoneRelative
    alphaBoolBlend {blendOption=Linear}
    boundBone {boneName=dyn_jacket_b_02}
    componentAppliedLinearAccClamp=<3500 3500 3500>
    componentLinearAccScale=<0.8 0.8 0.8>
    componentLinearVelScale=<0.8 0.8 0.8>
    physicsBodyDefinitions [{
      collisionType=InnerSphere
      sphereCollisionRadius=1
      boundBone {boneName=dyn_jacket_b_02}
      boxExtents=<10 20 20>
      constraintSetup {angularLimitsMax=<1 0 15>, angularLimitsMin=<-1 -30 -15>, angularTarget=<0 0 0>}
      localJointOffset=<2 0 0>
     }]
    relativeSpaceBone {boneName=dyn_root_simspace}
    sphericalLimits [
     {limitRadius=10, drivingBone {boneName=spine_01}, sphereLocalOffset=<0 0 0>}
     {limitRadius=11, drivingBone {boneName=pelvis}, sphereLocalOffset=<0 0 0>}
    ]
   }
   n37 = Get backpack() -> (backpack:double)
   n38 = Anim Dynamics - Bone: dyn_jacket_c_01!(ComponentPose=@n36, Alpha=@n37, GravityOverride:Vector=@n31, ExternalForce="0,0,0") -> (Pose) {
    angularDampingOverride=0.9999
    angularSpringConstant=50
    bAngularSpring=true
    bGravityOverrideInSimSpace=true
    bOverrideAngularDamping=true
    bUseGravityOverride=true
    bUseSphericalLimits=true
    simulationSpace=BoneRelative
    alphaBoolBlend {blendOption=Linear}
    boundBone {boneName=dyn_jacket_c_01}
    componentAppliedLinearAccClamp=<3500 3500 3500>
    componentLinearAccScale=<0.8 0.8 0.8>
    componentLinearVelScale=<0.8 0.8 0.8>
    physicsBodyDefinitions [{
      collisionType=InnerSphere
      sphereCollisionRadius=1
      boundBone {boneName=dyn_jacket_c_01}
      boxExtents=<10 20 20>
      constraintSetup {angularLimitsMax=<1 0 15>, angularLimitsMin=<-1 -30 -15>, angularTarget=<0 0 0>}
      localJointOffset=<2 0 0>
     }]
    relativeSpaceBone {boneName=dyn_root_simspace}
    sphericalLimits [{limitRadius=8, drivingBone {boneName=spine_03}, sphereLocalOffset=<0 0 -1>}]
   }
   n39 = Anim Dynamics - Bone: dyn_jacket_c_02!(ComponentPose=@n38, Alpha=@n37, GravityOverride:Vector=@n31, ExternalForce="0,0,0") -> (Pose) {
    angularDampingOverride=0.9999
    angularSpringConstant=50
    bAngularSpring=true
    bGravityOverrideInSimSpace=true
    bOverrideAngularDamping=true
    bUseGravityOverride=true
    bUseSphericalLimits=true
    simulationSpace=BoneRelative
    alphaBoolBlend {blendOption=Linear}
    boundBone {boneName=dyn_jacket_c_02}
    componentAppliedLinearAccClamp=<3500 3500 3500>
    componentLinearAccScale=<0.8 0.8 0.8>
    componentLinearVelScale=<0.8 0.8 0.8>
    physicsBodyDefinitions [{
      collisionType=InnerSphere
      sphereCollisionRadius=1
      boundBone {boneName=dyn_jacket_c_02}
      boxExtents=<10 20 20>
      constraintSetup {angularLimitsMax=<1 0 15>, angularLimitsMin=<-1 -30 -15>, angularTarget=<0 0 0>}
      localJointOffset=<2 0 0>
     }]
    relativeSpaceBone {boneName=dyn_root_simspace}
    sphericalLimits [{limitRadius=7.5, drivingBone {boneName=spine_01}, sphereLocalOffset=<2 0 0>}]
   }
   n40 = Anim Dynamics - Bone: dyn_jacket_d_01!(ComponentPose=@n39, Alpha=@n37, GravityOverride:Vector=@n31, ExternalForce="0,0,0") -> (Pose) {
    angularDampingOverride=0.9999
    angularSpringConstant=50
    bAngularSpring=true
    bGravityOverrideInSimSpace=true
    bOverrideAngularDamping=true
    bUseGravityOverride=true
    bUseSphericalLimits=true
    simulationSpace=BoneRelative
    alphaBoolBlend {blendOption=Linear}
    boundBone {boneName=dyn_jacket_d_01}
    componentAppliedLinearAccClamp=<3500 3500 3500>
    componentLinearAccScale=<0.8 0.8 0.8>
    componentLinearVelScale=<0.8 0.8 0.8>
    physicsBodyDefinitions [{
      collisionType=InnerSphere
      sphereCollisionRadius=1
      boundBone {boneName=dyn_jacket_d_01}
      boxExtents=<10 20 20>
      constraintSetup {angularLimitsMax=<1 0 15>, angularLimitsMin=<-1 -30 -15>, angularTarget=<0 0 0>}
      localJointOffset=<2 0 0>
     }]
    relativeSpaceBone {boneName=dyn_root_simspace}
    sphericalLimits [{limitRadius=7.5, drivingBone {boneName=spine_03}, sphereLocalOffset=<0 2 0>}]
   }
   n41 = Anim Dynamics - Bone: dyn_jacket_d_02!(ComponentPose=@n40, Alpha=@n37, GravityOverride:Vector=@n31, ExternalForce="0,0,0") -> (Pose) {
    angularDampingOverride=0.9999
    angularSpringConstant=50
    bAngularSpring=true
    bGravityOverrideInSimSpace=true
    bOverrideAngularDamping=true
    bUseGravityOverride=true
    bUseSphericalLimits=true
    simulationSpace=BoneRelative
    alphaBoolBlend {blendOption=Linear}
    boundBone {boneName=dyn_jacket_d_02}
    componentAppliedLinearAccClamp=<3500 3500 3500>
    componentLinearAccScale=<0.8 0.8 0.8>
    componentLinearVelScale=<0.8 0.8 0.8>
    physicsBodyDefinitions [{
      collisionType=InnerSphere
      sphereCollisionRadius=1
      boundBone {boneName=dyn_jacket_d_02}
      boxExtents=<10 20 20>
      constraintSetup {angularLimitsMax=<1 0 15>, angularLimitsMin=<-1 -30 -15>, angularTarget=<0 0 0>}
      localJointOffset=<2 0 0>
     }]
    relativeSpaceBone {boneName=dyn_root_simspace}
    sphericalLimits [{limitRadius=7, drivingBone {boneName=spine_01}, sphereLocalOffset=<3 2 0>}]
   }
   n42 = Anim Dynamics - Bone: dyn_jacket_e_01!(ComponentPose=@n41, Alpha=@n37, GravityOverride:Vector=@n31, ExternalForce="0,0,0") -> (Pose) {
    angularDampingOverride=0.9999
    angularSpringConstant=50
    bAngularSpring=true
    bGravityOverrideInSimSpace=true
    bOverrideAngularDamping=true
    bUseGravityOverride=true
    bUseSphericalLimits=true
    simulationSpace=BoneRelative
    alphaBoolBlend {blendOption=Linear}
    boundBone {boneName=dyn_jacket_e_01}
    componentAppliedLinearAccClamp=<3500 3500 3500>
    componentLinearAccScale=<0.8 0.8 0.8>
    componentLinearVelScale=<0.8 0.8 0.8>
    physicsBodyDefinitions [{
      collisionType=InnerSphere
      sphereCollisionRadius=1
      boundBone {boneName=dyn_jacket_e_01}
      boxExtents=<10 20 20>
      constraintSetup {angularLimitsMax=<1 0 15>, angularLimitsMin=<-1 -30 -15>, angularTarget=<0 0 0>}
      localJointOffset=<2 0 0>
     }]
    relativeSpaceBone {boneName=dyn_root_simspace}
    sphericalLimits [{limitRadius=7.5, drivingBone {boneName=spine_03}, sphereLocalOffset=<0 0 1>}]
   }
   n43 = Anim Dynamics - Bone: dyn_jacket_e_02!(ComponentPose=@n42, Alpha=@n37, GravityOverride:Vector=@n31, ExternalForce="0,0,0") -> (Pose) {
    angularDampingOverride=0.9999
    angularSpringConstant=50
    bAngularSpring=true
    bGravityOverrideInSimSpace=true
    bOverrideAngularDamping=true
    bUseGravityOverride=true
    bUseSphericalLimits=true
    simulationSpace=BoneRelative
    alphaBoolBlend {blendOption=Linear}
    boundBone {boneName=dyn_jacket_e_02}
    componentAppliedLinearAccClamp=<3500 3500 3500>
    componentLinearAccScale=<0.8 0.8 0.8>
    componentLinearVelScale=<0.8 0.8 0.8>
    physicsBodyDefinitions [{
      collisionType=InnerSphere
      sphereCollisionRadius=1
      boundBone {boneName=dyn_jacket_e_02}
      boxExtents=<10 20 20>
      constraintSetup {angularLimitsMax=<1 0 15>, angularLimitsMin=<-1 -30 -15>, angularTarget=<0 0 0>}
      localJointOffset=<2 0 0>
     }]
    relativeSpaceBone {boneName=dyn_root_simspace}
    sphericalLimits [{limitRadius=10, drivingBone {boneName=spine_01}, sphereLocalOffset=<0 3.5 0>}]
   }
   n44 = Anim Dynamics - Bone: dyn_jacket_f_01!(ComponentPose=@n43, Alpha=1, GravityOverride:Vector="0,0,0", ExternalForce="0,0,0") -> (Pose) {
    angularDampingOverride=0.9999
    angularSpringConstant=50
    bAngularSpring=true
    bGravityOverrideInSimSpace=true
    bOverrideAngularDamping=true
    simulationSpace=BoneRelative
    alphaBoolBlend {blendOption=Linear}
    boundBone {boneName=dyn_jacket_f_01}
    componentAppliedLinearAccClamp=<3500 3500 3500>
    componentLinearAccScale=<0.6 0.6 0.6>
    componentLinearVelScale=<0.6 0.6 0.6>
    physicsBodyDefinitions [{
      sphereCollisionRadius=0
      boundBone {boneName=dyn_jacket_f_01}
      boxExtents=<10 20 20>
      constraintSetup {angularLimitsMax=<1 0 5>, angularLimitsMin=<-1 -30 -5>, angularTarget=<0 0 0>}
      localJointOffset=<2 0 0>
     }]
    relativeSpaceBone {boneName=dyn_root_simspace}
   }
   n45 = Anim Dynamics - Bone: dyn_jacket_f_02!(ComponentPose=@n44, Alpha=1, GravityOverride:Vector="0,0,0", ExternalForce="0,0,0") -> (Pose) {
    angularDampingOverride=0.9999
    angularSpringConstant=50
    bAngularSpring=true
    bGravityOverrideInSimSpace=true
    bOverrideAngularDamping=true
    bUseSphericalLimits=true
    simulationSpace=BoneRelative
    alphaBoolBlend {blendOption=Linear}
    boundBone {boneName=dyn_jacket_f_02}
    componentAppliedLinearAccClamp=<3500 3500 3500>
    componentLinearAccScale=<0.6 0.6 0.6>
    componentLinearVelScale=<0.6 0.6 0.6>
    physicsBodyDefinitions [{
      collisionType=InnerSphere
      sphereCollisionRadius=1
      boundBone {boneName=dyn_jacket_f_02}
      boxExtents=<10 20 20>
      constraintSetup {angularLimitsMax=<1 0 5>, angularLimitsMin=<-1 -30 -5>, angularTarget=<0 0 0>}
      localJointOffset=<2 0 0>
     }]
    relativeSpaceBone {boneName=dyn_root_simspace}
    sphericalLimits [
     {limitRadius=9, drivingBone {boneName=spine_01}, sphereLocalOffset=<0 0 1.5>}
     {limitRadius=12, drivingBone {boneName=pelvis}, sphereLocalOffset=<0 0 0.5>}
    ]
   }
   n46 = Anim Dynamics - Bone: dyn_jacket_g_01!(ComponentPose=@n45, Alpha=1, GravityOverride:Vector="0,0,0", ExternalForce="0,0,0") -> (Pose) {
    angularDampingOverride=0.9999
    angularSpringConstant=50
    bAngularSpring=true
    bGravityOverrideInSimSpace=true
    bOverrideAngularDamping=true
    simulationSpace=BoneRelative
    alphaBoolBlend {blendOption=Linear}
    boundBone {boneName=dyn_jacket_g_01}
    componentAppliedLinearAccClamp=<3500 3500 3500>
    componentLinearAccScale=<0.6 0.6 0.6>
    componentLinearVelScale=<0.6 0.6 0.6>
    physicsBodyDefinitions [{
      sphereCollisionRadius=0
      boundBone {boneName=dyn_jacket_g_01}
      boxExtents=<10 20 20>
      constraintSetup {angularLimitsMax=<1 -2 5>, angularLimitsMin=<-1 -15 -5>, angularTarget=<0 0 0>}
      localJointOffset=<2 0 0>
     }]
    relativeSpaceBone {boneName=dyn_root_simspace}
   }
   n47 = Anim Dynamics - Bone: dyn_jacket_g_02!(ComponentPose=@n46, Alpha=1, GravityOverride:Vector="0,0,0", ExternalForce="0,0,0") -> (Pose) {
    angularDampingOverride=0.9999
    angularSpringConstant=50
    bAngularSpring=true
    bGravityOverrideInSimSpace=true
    bOverrideAngularDamping=true
    bUseSphericalLimits=true
    simulationSpace=BoneRelative
    alphaBoolBlend {blendOption=Linear}
    boundBone {boneName=dyn_jacket_g_02}
    componentAppliedLinearAccClamp=<3500 3500 3500>
    componentLinearAccScale=<0.6 0.6 0.6>
    componentLinearVelScale=<0.6 0.6 0.6>
    physicsBodyDefinitions [{
      collisionType=CustomSphere
      sphereCollisionRadius=1
      boundBone {boneName=dyn_jacket_g_02}
      boxExtents=<10 20 20>
      constraintSetup {angularLimitsMax=<1 -2 5>, angularLimitsMin=<-1 -15 -5>, angularTarget=<0 0 0>}
      localJointOffset=<2 0 0>
     }]
    relativeSpaceBone {boneName=dyn_root_simspace}
   }
   n48 = Anim Dynamics - Bone: dyn_jacket_g_03!(ComponentPose=@n47, Alpha=1, GravityOverride:Vector="0,0,0", ExternalForce="0,0,0") -> (Pose) {
    angularDampingOverride=0.9999
    angularSpringConstant=50
    bAngularSpring=true
    bGravityOverrideInSimSpace=true
    bOverrideAngularDamping=true
    bUseSphericalLimits=true
    simulationSpace=BoneRelative
    alphaBoolBlend {blendOption=Linear}
    boundBone {boneName=dyn_jacket_g_03}
    componentAppliedLinearAccClamp=<3500 3500 3500>
    componentLinearAccScale=<0.6 0.6 0.6>
    componentLinearVelScale=<0.6 0.6 0.6>
    physicsBodyDefinitions [{
      collisionType=InnerSphere
      sphereCollisionRadius=1
      boundBone {boneName=dyn_jacket_g_03}
      boxExtents=<10 20 20>
      constraintSetup {angularLimitsMax=<1 -2 5>, angularLimitsMin=<-1 -15 -5>, angularTarget=<0 0 0>}
      localJointOffset=<2 0 0>
     }]
    relativeSpaceBone {boneName=dyn_root_simspace}
    sphericalLimits [
     {limitRadius=9.5, drivingBone {boneName=spine_01}, sphereLocalOffset=<0 0 0>}
     {limitRadius=10, drivingBone {boneName=spine_02}, sphereLocalOffset=<0 0 0>}
     {limitRadius=11, drivingBone {boneName=pelvis}, sphereLocalOffset=<0 0 0>}
    ]
   }
   n49 = Get jacket dynamics()
   n50 = Get GravityOverride_2()
   n51 = Anim Dynamics - Bone: dyn_jacket_a_01(ComponentPose=Local To Component(LocalPose=Use cached pose 'prefrontend'()), Alpha=@n49, GravityOverride=@n50, ExternalForce="0,0,0") {
    angularDampingOverride=0.9999
    angularSpringConstant=50
    bAngularSpring=true
    bGravityOverrideInSimSpace=true
    bOverrideAngularDamping=true
    bUseGravityOverride=true
    bUsePlanarLimit=false
    simulationSpace=BoneRelative
    alphaBoolBlend {blendOption=Linear}
    boundBone {boneName=dyn_jacket_a_01}
    componentAppliedLinearAccClamp=<3500 3500 3500>
    componentLinearAccScale=<1 1 1>
    componentLinearVelScale=<1 1 1>
    physicsBodyDefinitions [{
      sphereCollisionRadius=0
      boundBone {boneName=dyn_jacket_a_01}
      boxExtents=<10 20 20>
      constraintSetup {angularLimitsMax=<1 0 25>, angularLimitsMin=<-1 -25 -25>, angularTarget=<0 0 0>}
      localJointOffset=<2 0 0>
     }]
    relativeSpaceBone {boneName=dyn_root_simspace}
   }
   n52 = Anim Dynamics - Bone: dyn_jacket_a_02(ComponentPose=@n51, Alpha=@n49, GravityOverride=@n50, ExternalForce="0,0,0") {
    angularDampingOverride=0.9999
    angularSpringConstant=50
    bAngularSpring=true
    bGravityOverrideInSimSpace=true
    bOverrideAngularDamping=true
    bUseGravityOverride=true
    bUsePlanarLimit=false
    simulationSpace=BoneRelative
    alphaBoolBlend {blendOption=Linear}
    boundBone {boneName=dyn_jacket_a_02}
    componentAppliedLinearAccClamp=<3500 3500 3500>
    componentLinearAccScale=<1 1 1>
    componentLinearVelScale=<1 1 1>
    physicsBodyDefinitions [{
      collisionType=InnerSphere
      sphereCollisionRadius=0
      boundBone {boneName=dyn_jacket_a_02}
      boxExtents=<10 20 20>
      constraintSetup {angularLimitsMax=<1 0 25>, angularLimitsMin=<-1 -25 -25>, angularTarget=<0 0 0>}
      localJointOffset=<2 0 0>
     }]
    relativeSpaceBone {boneName=dyn_root_simspace}
   }
   n53 = Anim Dynamics - Bone: dyn_jacket_a_03(ComponentPose=@n52, Alpha=@n49, GravityOverride=@n50, ExternalForce="0,0,0") {
    angularDampingOverride=0.9999
    angularSpringConstant=50
    bAngularSpring=true
    bGravityOverrideInSimSpace=true
    bOverrideAngularDamping=true
    bUseGravityOverride=true
    bUsePlanarLimit=false
    bUseSphericalLimits=true
    simulationSpace=BoneRelative
    alphaBoolBlend {blendOption=Linear}
    boundBone {boneName=dyn_jacket_a_03}
    componentAppliedLinearAccClamp=<3500 3500 3500>
    componentLinearAccScale=<1 1 1>
    componentLinearVelScale=<1 1 1>
    physicsBodyDefinitions [{
      collisionType=InnerSphere
      sphereCollisionRadius=1
      boundBone {boneName=dyn_jacket_a_03}
      boxExtents=<10 20 20>
      constraintSetup {angularLimitsMax=<1 0 25>, angularLimitsMin=<-1 -25 -25>, angularTarget=<0 0 0>}
      localJointOffset=<2 0 0>
     }]
    relativeSpaceBone {boneName=dyn_root_simspace}
    sphericalLimits [
     {limitRadius=5.7, drivingBone {boneName=spine_01}, sphereLocalOffset=<0 0 -1>}
     {limitRadius=6.5, drivingBone {boneName=pelvis}, sphereLocalOffset=<0 0 0>}
    ]
   }
   n54 = Anim Dynamics - Bone: dyn_jacket_b_01(ComponentPose=@n53, Alpha=1, GravityOverride=@n50, ExternalForce="0,0,0") {
    angularDampingOverride=0.9999
    angularSpringConstant=50
    bAngularSpring=true
    bGravityOverrideInSimSpace=true
    bOverrideAngularDamping=true
    bUseGravityOverride=true
    simulationSpace=BoneRelative
    alphaBoolBlend {blendOption=Linear}
    boundBone {boneName=dyn_jacket_b_01}
    componentAppliedLinearAccClamp=<3500 3500 3500>
    componentLinearAccScale=<0.8 0.8 0.8>
    componentLinearVelScale=<0.8 0.8 0.8>
    physicsBodyDefinitions [{
      sphereCollisionRadius=0
      boundBone {boneName=dyn_jacket_b_01}
      boxExtents=<10 20 20>
      constraintSetup {angularLimitsMax=<1 0 15>, angularLimitsMin=<-1 -30 -15>, angularTarget=<0 0 0>}
      localJointOffset=<2 0 0>
     }]
    relativeSpaceBone {boneName=dyn_root_simspace}
   }
   n55 = Anim Dynamics - Bone: dyn_jacket_b_02(ComponentPose=@n54, Alpha=1, GravityOverride=@n50, ExternalForce="0,0,0") {
    angularDampingOverride=0.9999
    angularSpringConstant=50
    bAngularSpring=true
    bGravityOverrideInSimSpace=true
    bOverrideAngularDamping=true
    bUseGravityOverride=true
    bUseSphericalLimits=true
    simulationSpace=BoneRelative
    alphaBoolBlend {blendOption=Linear}
    boundBone {boneName=dyn_jacket_b_02}
    componentAppliedLinearAccClamp=<3500 3500 3500>
    componentLinearAccScale=<0.8 0.8 0.8>
    componentLinearVelScale=<0.8 0.8 0.8>
    physicsBodyDefinitions [{
      collisionType=InnerSphere
      sphereCollisionRadius=1
      boundBone {boneName=dyn_jacket_b_02}
      boxExtents=<10 20 20>
      constraintSetup {angularLimitsMax=<1 0 15>, angularLimitsMin=<-1 -30 -15>, angularTarget=<0 0 0>}
      localJointOffset=<2 0 0>
     }]
    relativeSpaceBone {boneName=dyn_root_simspace}
    sphericalLimits [
     {limitRadius=10, drivingBone {boneName=spine_01}, sphereLocalOffset=<0 0 0>}
     {limitRadius=11, drivingBone {boneName=pelvis}, sphereLocalOffset=<0 0 0>}
    ]
   }
   n56 = Get backpack()
   n57 = Anim Dynamics - Bone: dyn_jacket_c_01(ComponentPose=@n55, Alpha=@n56, GravityOverride=@n50, ExternalForce="0,0,0") {
    angularDampingOverride=0.9999
    angularSpringConstant=50
    bAngularSpring=true
    bGravityOverrideInSimSpace=true
    bOverrideAngularDamping=true
    bUseGravityOverride=true
    bUseSphericalLimits=true
    simulationSpace=BoneRelative
    alphaBoolBlend {blendOption=Linear}
    boundBone {boneName=dyn_jacket_c_01}
    componentAppliedLinearAccClamp=<3500 3500 3500>
    componentLinearAccScale=<0.8 0.8 0.8>
    componentLinearVelScale=<0.8 0.8 0.8>
    physicsBodyDefinitions [{
      collisionType=InnerSphere
      sphereCollisionRadius=1
      boundBone {boneName=dyn_jacket_c_01}
      boxExtents=<10 20 20>
      constraintSetup {angularLimitsMax=<1 0 15>, angularLimitsMin=<-1 -30 -15>, angularTarget=<0 0 0>}
      localJointOffset=<2 0 0>
     }]
    relativeSpaceBone {boneName=dyn_root_simspace}
    sphericalLimits [{limitRadius=8, drivingBone {boneName=spine_03}, sphereLocalOffset=<0 0 -1>}]
   }
   n58 = Anim Dynamics - Bone: dyn_jacket_c_02(ComponentPose=@n57, Alpha=@n56, GravityOverride=@n50, ExternalForce="0,0,0") {
    angularDampingOverride=0.9999
    angularSpringConstant=50
    bAngularSpring=true
    bGravityOverrideInSimSpace=true
    bOverrideAngularDamping=true
    bUseGravityOverride=true
    bUseSphericalLimits=true
    simulationSpace=BoneRelative
    alphaBoolBlend {blendOption=Linear}
    boundBone {boneName=dyn_jacket_c_02}
    componentAppliedLinearAccClamp=<3500 3500 3500>
    componentLinearAccScale=<0.8 0.8 0.8>
    componentLinearVelScale=<0.8 0.8 0.8>
    physicsBodyDefinitions [{
      collisionType=InnerSphere
      sphereCollisionRadius=1
      boundBone {boneName=dyn_jacket_c_02}
      boxExtents=<10 20 20>
      constraintSetup {angularLimitsMax=<1 0 15>, angularLimitsMin=<-1 -30 -15>, angularTarget=<0 0 0>}
      localJointOffset=<2 0 0>
     }]
    relativeSpaceBone {boneName=dyn_root_simspace}
    sphericalLimits [{limitRadius=7.5, drivingBone {boneName=spine_01}, sphereLocalOffset=<2 0 0>}]
   }
   n59 = Anim Dynamics - Bone: dyn_jacket_d_01(ComponentPose=@n58, Alpha=@n56, GravityOverride=@n50, ExternalForce="0,0,0") {
    angularDampingOverride=0.9999
    angularSpringConstant=50
    bAngularSpring=true
    bGravityOverrideInSimSpace=true
    bOverrideAngularDamping=true
    bUseGravityOverride=true
    bUseSphericalLimits=true
    simulationSpace=BoneRelative
    alphaBoolBlend {blendOption=Linear}
    boundBone {boneName=dyn_jacket_d_01}
    componentAppliedLinearAccClamp=<3500 3500 3500>
    componentLinearAccScale=<0.8 0.8 0.8>
    componentLinearVelScale=<0.8 0.8 0.8>
    physicsBodyDefinitions [{
      collisionType=InnerSphere
      sphereCollisionRadius=1
      boundBone {boneName=dyn_jacket_d_01}
      boxExtents=<10 20 20>
      constraintSetup {angularLimitsMax=<1 0 15>, angularLimitsMin=<-1 -30 -15>, angularTarget=<0 0 0>}
      localJointOffset=<2 0 0>
     }]
    relativeSpaceBone {boneName=dyn_root_simspace}
    sphericalLimits [{limitRadius=7.5, drivingBone {boneName=spine_03}, sphereLocalOffset=<0 2 0>}]
   }
   n60 = Anim Dynamics - Bone: dyn_jacket_d_02(ComponentPose=@n59, Alpha=@n56, GravityOverride=@n50, ExternalForce="0,0,0") {
    angularDampingOverride=0.9999
    angularSpringConstant=50
    bAngularSpring=true
    bGravityOverrideInSimSpace=true
    bOverrideAngularDamping=true
    bUseGravityOverride=true
    bUseSphericalLimits=true
    simulationSpace=BoneRelative
    alphaBoolBlend {blendOption=Linear}
    boundBone {boneName=dyn_jacket_d_02}
    componentAppliedLinearAccClamp=<3500 3500 3500>
    componentLinearAccScale=<0.8 0.8 0.8>
    componentLinearVelScale=<0.8 0.8 0.8>
    physicsBodyDefinitions [{
      collisionType=InnerSphere
      sphereCollisionRadius=1
      boundBone {boneName=dyn_jacket_d_02}
      boxExtents=<10 20 20>
      constraintSetup {angularLimitsMax=<1 0 15>, angularLimitsMin=<-1 -30 -15>, angularTarget=<0 0 0>}
      localJointOffset=<2 0 0>
     }]
    relativeSpaceBone {boneName=dyn_root_simspace}
    sphericalLimits [{limitRadius=7, drivingBone {boneName=spine_01}, sphereLocalOffset=<3 2 0>}]
   }
   n61 = Anim Dynamics - Bone: dyn_jacket_e_01(ComponentPose=@n60, Alpha=@n56, GravityOverride=@n50, ExternalForce="0,0,0") {
    angularDampingOverride=0.9999
    angularSpringConstant=50
    bAngularSpring=true
    bGravityOverrideInSimSpace=true
    bOverrideAngularDamping=true
    bUseGravityOverride=true
    bUseSphericalLimits=true
    simulationSpace=BoneRelative
    alphaBoolBlend {blendOption=Linear}
    boundBone {boneName=dyn_jacket_e_01}
    componentAppliedLinearAccClamp=<3500 3500 3500>
    componentLinearAccScale=<0.8 0.8 0.8>
    componentLinearVelScale=<0.8 0.8 0.8>
    physicsBodyDefinitions [{
      collisionType=InnerSphere
      sphereCollisionRadius=1
      boundBone {boneName=dyn_jacket_e_01}
      boxExtents=<10 20 20>
      constraintSetup {angularLimitsMax=<1 0 15>, angularLimitsMin=<-1 -30 -15>, angularTarget=<0 0 0>}
      localJointOffset=<2 0 0>
     }]
    relativeSpaceBone {boneName=dyn_root_simspace}
    sphericalLimits [{limitRadius=7.5, drivingBone {boneName=spine_03}, sphereLocalOffset=<0 0 1>}]
   }
   n62 = Anim Dynamics - Bone: dyn_jacket_e_02(ComponentPose=@n61, Alpha=@n56, GravityOverride=@n50, ExternalForce="0,0,0") {
    angularDampingOverride=0.9999
    angularSpringConstant=50
    bAngularSpring=true
    bGravityOverrideInSimSpace=true
    bOverrideAngularDamping=true
    bUseGravityOverride=true
    bUseSphericalLimits=true
    simulationSpace=BoneRelative
    alphaBoolBlend {blendOption=Linear}
    boundBone {boneName=dyn_jacket_e_02}
    componentAppliedLinearAccClamp=<3500 3500 3500>
    componentLinearAccScale=<0.8 0.8 0.8>
    componentLinearVelScale=<0.8 0.8 0.8>
    physicsBodyDefinitions [{
      collisionType=InnerSphere
      sphereCollisionRadius=1
      boundBone {boneName=dyn_jacket_e_02}
      boxExtents=<10 20 20>
      constraintSetup {angularLimitsMax=<1 0 15>, angularLimitsMin=<-1 -30 -15>, angularTarget=<0 0 0>}
      localJointOffset=<2 0 0>
     }]
    relativeSpaceBone {boneName=dyn_root_simspace}
    sphericalLimits [{limitRadius=10, drivingBone {boneName=spine_01}, sphereLocalOffset=<0 3.5 0>}]
   }
   n63 = Anim Dynamics - Bone: dyn_jacket_f_01(ComponentPose=@n62, Alpha=1, GravityOverride="0,0,0", ExternalForce="0,0,0") {
    angularDampingOverride=0.9999
    angularSpringConstant=50
    bAngularSpring=true
    bGravityOverrideInSimSpace=true
    bOverrideAngularDamping=true
    simulationSpace=BoneRelative
    alphaBoolBlend {blendOption=Linear}
    boundBone {boneName=dyn_jacket_f_01}
    componentAppliedLinearAccClamp=<3500 3500 3500>
    componentLinearAccScale=<0.6 0.6 0.6>
    componentLinearVelScale=<0.6 0.6 0.6>
    physicsBodyDefinitions [{
      sphereCollisionRadius=0
      boundBone {boneName=dyn_jacket_f_01}
      boxExtents=<10 20 20>
      constraintSetup {angularLimitsMax=<1 0 5>, angularLimitsMin=<-1 -30 -5>, angularTarget=<0 0 0>}
      localJointOffset=<2 0 0>
     }]
    relativeSpaceBone {boneName=dyn_root_simspace}
   }
   n64 = Anim Dynamics - Bone: dyn_jacket_f_02(ComponentPose=@n63, Alpha=1, GravityOverride="0,0,0", ExternalForce="0,0,0") {
    angularDampingOverride=0.9999
    angularSpringConstant=50
    bAngularSpring=true
    bGravityOverrideInSimSpace=true
    bOverrideAngularDamping=true
    bUseSphericalLimits=true
    simulationSpace=BoneRelative
    alphaBoolBlend {blendOption=Linear}
    boundBone {boneName=dyn_jacket_f_02}
    componentAppliedLinearAccClamp=<3500 3500 3500>
    componentLinearAccScale=<0.6 0.6 0.6>
    componentLinearVelScale=<0.6 0.6 0.6>
    physicsBodyDefinitions [{
      collisionType=InnerSphere
      sphereCollisionRadius=1
      boundBone {boneName=dyn_jacket_f_02}
      boxExtents=<10 20 20>
      constraintSetup {angularLimitsMax=<1 0 5>, angularLimitsMin=<-1 -30 -5>, angularTarget=<0 0 0>}
      localJointOffset=<2 0 0>
     }]
    relativeSpaceBone {boneName=dyn_root_simspace}
    sphericalLimits [
     {limitRadius=9, drivingBone {boneName=spine_01}, sphereLocalOffset=<0 0 1.5>}
     {limitRadius=12, drivingBone {boneName=pelvis}, sphereLocalOffset=<0 0 0.5>}
    ]
   }
   n65 = Anim Dynamics - Bone: dyn_jacket_g_01(ComponentPose=@n64, Alpha=1, GravityOverride="0,0,0", ExternalForce="0,0,0") {
    angularDampingOverride=0.9999
    angularSpringConstant=50
    bAngularSpring=true
    bGravityOverrideInSimSpace=true
    bOverrideAngularDamping=true
    bUsePlanarLimit=false
    simulationSpace=BoneRelative
    alphaBoolBlend {blendOption=Linear}
    boundBone {boneName=dyn_jacket_g_01}
    componentAppliedLinearAccClamp=<3500 3500 3500>
    componentLinearAccScale=<0.6 0.6 0.6>
    componentLinearVelScale=<0.6 0.6 0.6>
    physicsBodyDefinitions [{
      sphereCollisionRadius=0
      boundBone {boneName=dyn_jacket_g_01}
      boxExtents=<10 20 20>
      constraintSetup {angularLimitsMax=<1 -2 5>, angularLimitsMin=<-1 -15 -5>, angularTarget=<0 0 0>}
      localJointOffset=<2 0 0>
     }]
    relativeSpaceBone {boneName=dyn_root_simspace}
   }
   n66 = Anim Dynamics - Bone: dyn_jacket_g_02(ComponentPose=@n65, Alpha=1, GravityOverride="0,0,0", ExternalForce="0,0,0") {
    angularDampingOverride=0.9999
    angularSpringConstant=50
    bAngularSpring=true
    bGravityOverrideInSimSpace=true
    bOverrideAngularDamping=true
    bUsePlanarLimit=false
    simulationSpace=BoneRelative
    alphaBoolBlend {blendOption=Linear}
    boundBone {boneName=dyn_jacket_g_02}
    componentAppliedLinearAccClamp=<3500 3500 3500>
    componentLinearAccScale=<0.6 0.6 0.6>
    componentLinearVelScale=<0.6 0.6 0.6>
    physicsBodyDefinitions [{
      collisionType=CustomSphere
      sphereCollisionRadius=1
      boundBone {boneName=dyn_jacket_g_02}
      boxExtents=<10 20 20>
      constraintSetup {angularLimitsMax=<1 -2 5>, angularLimitsMin=<-1 -15 -5>, angularTarget=<0 0 0>}
      localJointOffset=<2 0 0>
     }]
    relativeSpaceBone {boneName=dyn_root_simspace}
   }
   n67 = Anim Dynamics - Bone: dyn_jacket_g_03(ComponentPose=@n66, Alpha=1, GravityOverride="0,0,0", ExternalForce="0,0,0") {
    angularDampingOverride=0.9999
    angularSpringConstant=50
    bAngularSpring=true
    bGravityOverrideInSimSpace=true
    bOverrideAngularDamping=true
    bUsePlanarLimit=false
    simulationSpace=BoneRelative
    alphaBoolBlend {blendOption=Linear}
    boundBone {boneName=dyn_jacket_g_03}
    componentAppliedLinearAccClamp=<3500 3500 3500>
    componentLinearAccScale=<0.6 0.6 0.6>
    componentLinearVelScale=<0.6 0.6 0.6>
    physicsBodyDefinitions [{
      collisionType=InnerSphere
      sphereCollisionRadius=1
      boundBone {boneName=dyn_jacket_g_03}
      boxExtents=<10 20 20>
      constraintSetup {angularLimitsMax=<1 -2 5>, angularLimitsMin=<-1 -15 -5>, angularTarget=<0 0 0>}
      localJointOffset=<2 0 0>
     }]
    relativeSpaceBone {boneName=dyn_root_simspace}
    sphericalLimits [
     {limitRadius=9.5, drivingBone {boneName=spine_01}, sphereLocalOffset=<0 0 0>}
     {limitRadius=10, drivingBone {boneName=spine_02}, sphereLocalOffset=<0 0 0>}
     {limitRadius=11, drivingBone {boneName=pelvis}, sphereLocalOffset=<0 0 0>}
    ]
   }
   n68 = Blend Poses by bool(BlendPose_0=Component To Local(ComponentPose=@n48), BlendPose_1=Component To Local(ComponentPose=@n67), bActiveValue=Get bIsPlayingEmote()) {blendType=Linear}
   n69 = Get jacket dynamics()
   n70 = Get GravityOverride_2()
   n71 = Anim Dynamics - Bone: dyn_jacket_a_01(ComponentPose=Local To Component(LocalPose=Use cached pose 'prefrontend'()), Alpha=@n69, GravityOverride=@n70, ExternalForce="0,0,0") {
    angularDampingOverride=0.9999
    angularSpringConstant=50
    bAngularSpring=true
    bGravityOverrideInSimSpace=true
    bOverrideAngularDamping=true
    bUseGravityOverride=true
    simulationSpace=BoneRelative
    alphaBoolBlend {blendOption=Linear}
    boundBone {boneName=dyn_jacket_a_01}
    componentAppliedLinearAccClamp=<3500 3500 3500>
    componentLinearAccScale=<1 1 1>
    componentLinearVelScale=<1 1 1>
    physicsBodyDefinitions [{
      sphereCollisionRadius=0
      boundBone {boneName=dyn_jacket_a_01}
      boxExtents=<10 20 20>
      constraintSetup {angularLimitsMax=<1 0 10>, angularLimitsMin=<-1 -25 -25>, angularTarget=<0 0 0>}
      localJointOffset=<2 0 0>
     }]
    relativeSpaceBone {boneName=dyn_root_simspace}
   }
   n72 = Anim Dynamics - Bone: dyn_jacket_a_02(ComponentPose=@n71, Alpha=@n69, GravityOverride=@n70, ExternalForce="0,0,0") {
    angularDampingOverride=0.9999
    angularSpringConstant=50
    bAngularSpring=true
    bGravityOverrideInSimSpace=true
    bOverrideAngularDamping=true
    bUseGravityOverride=true
    simulationSpace=BoneRelative
    alphaBoolBlend {blendOption=Linear}
    boundBone {boneName=dyn_jacket_a_02}
    componentAppliedLinearAccClamp=<3500 3500 3500>
    componentLinearAccScale=<1 1 1>
    componentLinearVelScale=<1 1 1>
    physicsBodyDefinitions [{
      sphereCollisionRadius=0
      boundBone {boneName=dyn_jacket_a_02}
      boxExtents=<10 20 20>
      constraintSetup {angularLimitsMax=<1 0 10>, angularLimitsMin=<-1 -25 -25>, angularTarget=<0 0 0>}
      localJointOffset=<2 0 0>
     }]
    relativeSpaceBone {boneName=dyn_root_simspace}
   }
   n73 = Anim Dynamics - Bone: dyn_jacket_a_03(ComponentPose=@n72, Alpha=@n69, GravityOverride=@n70, ExternalForce="0,0,0") {
    angularDampingOverride=0.9999
    angularSpringConstant=50
    bAngularSpring=true
    bGravityOverrideInSimSpace=true
    bOverrideAngularDamping=true
    bUseGravityOverride=true
    simulationSpace=BoneRelative
    alphaBoolBlend {blendOption=Linear}
    boundBone {boneName=dyn_jacket_a_03}
    componentAppliedLinearAccClamp=<3500 3500 3500>
    componentLinearAccScale=<1 1 1>
    componentLinearVelScale=<1 1 1>
    physicsBodyDefinitions [{
      collisionType=InnerSphere
      sphereCollisionRadius=1
      boundBone {boneName=dyn_jacket_a_03}
      boxExtents=<10 20 20>
      constraintSetup {angularLimitsMax=<1 0 10>, angularLimitsMin=<-1 -25 -25>, angularTarget=<0 0 0>}
      localJointOffset=<2 0 0>
     }]
    relativeSpaceBone {boneName=dyn_root_simspace}
    sphericalLimits [
     {limitRadius=15, drivingBone {boneName=spine_01}, sphereLocalOffset=<0 -5 6>}
     {limitRadius=15, drivingBone {boneName=pelvis}, sphereLocalOffset=<0 -5 5>}
    ]
   }
   n74 = Anim Dynamics - Bone: dyn_jacket_b_01(ComponentPose=@n73, Alpha=1, GravityOverride=@n70, ExternalForce="0,0,0") {
    angularDampingOverride=0.9999
    angularSpringConstant=50
    bAngularSpring=true
    bGravityOverrideInSimSpace=true
    bOverrideAngularDamping=true
    bUseGravityOverride=true
    simulationSpace=BoneRelative
    alphaBoolBlend {blendOption=Linear}
    boundBone {boneName=dyn_jacket_b_01}
    componentAppliedLinearAccClamp=<3500 3500 3500>
    componentLinearAccScale=<0.8 0.8 0.8>
    componentLinearVelScale=<0.8 0.8 0.8>
    physicsBodyDefinitions [{
      sphereCollisionRadius=0
      boundBone {boneName=dyn_jacket_b_01}
      boxExtents=<10 20 20>
      constraintSetup {angularLimitsMax=<1 0 15>, angularLimitsMin=<-1 -30 -15>, angularTarget=<0 0 0>}
      localJointOffset=<2 0 0>
     }]
    relativeSpaceBone {boneName=dyn_root_simspace}
   }
   n75 = Anim Dynamics - Bone: dyn_jacket_b_02(ComponentPose=@n74, Alpha=1, GravityOverride=@n70, ExternalForce="0,0,0") {
    angularDampingOverride=0.9999
    angularSpringConstant=50
    bAngularSpring=true
    bGravityOverrideInSimSpace=true
    bOverrideAngularDamping=true
    bUseGravityOverride=true
    bUseSphericalLimits=true
    simulationSpace=BoneRelative
    alphaBoolBlend {blendOption=Linear}
    boundBone {boneName=dyn_jacket_b_02}
    componentAppliedLinearAccClamp=<3500 3500 3500>
    componentLinearAccScale=<0.8 0.8 0.8>
    componentLinearVelScale=<0.8 0.8 0.8>
    physicsBodyDefinitions [{
      collisionType=InnerSphere
      sphereCollisionRadius=1
      boundBone {boneName=dyn_jacket_b_02}
      boxExtents=<10 20 20>
      constraintSetup {angularLimitsMax=<1 0 15>, angularLimitsMin=<-1 -30 -15>, angularTarget=<0 0 0>}
      localJointOffset=<2 0 0>
     }]
    relativeSpaceBone {boneName=dyn_root_simspace}
    sphericalLimits [
     {limitRadius=15, drivingBone {boneName=spine_01}, sphereLocalOffset=<0 0 6>}
     {limitRadius=15, drivingBone {boneName=pelvis}, sphereLocalOffset=<0 0 5>}
    ]
   }
   n76 = Get backpack()
   n77 = Anim Dynamics - Bone: dyn_jacket_c_01(ComponentPose=@n75, Alpha=@n76, GravityOverride=@n70, ExternalForce="0,0,0") {
    angularDampingOverride=0.9999
    angularSpringConstant=50
    bAngularSpring=true
    bGravityOverrideInSimSpace=true
    bOverrideAngularDamping=true
    bUseGravityOverride=true
    bUseSphericalLimits=true
    simulationSpace=BoneRelative
    alphaBoolBlend {blendOption=Linear}
    boundBone {boneName=dyn_jacket_c_01}
    componentAppliedLinearAccClamp=<3500 3500 3500>
    componentLinearAccScale=<0.8 0.8 0.8>
    componentLinearVelScale=<0.8 0.8 0.8>
    physicsBodyDefinitions [{
      collisionType=InnerSphere
      sphereCollisionRadius=1
      boundBone {boneName=dyn_jacket_c_01}
      boxExtents=<10 20 20>
      constraintSetup {angularLimitsMax=<1 0 15>, angularLimitsMin=<-1 -30 -15>, angularTarget=<0 0 0>}
      localJointOffset=<2 0 0>
     }]
    relativeSpaceBone {boneName=dyn_root_simspace}
    sphericalLimits [{limitRadius=8, drivingBone {boneName=spine_03}, sphereLocalOffset=<0 0 -1>}]
   }
   n78 = Anim Dynamics - Bone: dyn_jacket_c_02(ComponentPose=@n77, Alpha=@n76, GravityOverride=@n70, ExternalForce="0,0,0") {
    angularDampingOverride=0.9999
    angularSpringConstant=50
    bAngularSpring=true
    bGravityOverrideInSimSpace=true
    bOverrideAngularDamping=true
    bUseGravityOverride=true
    bUseSphericalLimits=true
    simulationSpace=BoneRelative
    alphaBoolBlend {blendOption=Linear}
    boundBone {boneName=dyn_jacket_c_02}
    componentAppliedLinearAccClamp=<3500 3500 3500>
    componentLinearAccScale=<0.8 0.8 0.8>
    componentLinearVelScale=<0.8 0.8 0.8>
    physicsBodyDefinitions [{
      collisionType=InnerSphere
      sphereCollisionRadius=1
      boundBone {boneName=dyn_jacket_c_02}
      boxExtents=<10 20 20>
      constraintSetup {angularLimitsMax=<1 0 15>, angularLimitsMin=<-1 -30 -15>, angularTarget=<0 0 0>}
      localJointOffset=<2 0 0>
     }]
    relativeSpaceBone {boneName=dyn_root_simspace}
    sphericalLimits [{limitRadius=7.5, drivingBone {boneName=spine_01}, sphereLocalOffset=<2 0 0>}]
   }
   n79 = Anim Dynamics - Bone: dyn_jacket_d_01(ComponentPose=@n78, Alpha=@n76, GravityOverride=@n70, ExternalForce="0,0,0") {
    angularDampingOverride=0.9999
    angularSpringConstant=50
    bAngularSpring=true
    bGravityOverrideInSimSpace=true
    bOverrideAngularDamping=true
    bUseGravityOverride=true
    bUseSphericalLimits=true
    simulationSpace=BoneRelative
    alphaBoolBlend {blendOption=Linear}
    boundBone {boneName=dyn_jacket_d_01}
    componentAppliedLinearAccClamp=<3500 3500 3500>
    componentLinearAccScale=<0.8 0.8 0.8>
    componentLinearVelScale=<0.8 0.8 0.8>
    physicsBodyDefinitions [{
      collisionType=InnerSphere
      sphereCollisionRadius=1
      boundBone {boneName=dyn_jacket_d_01}
      boxExtents=<10 20 20>
      constraintSetup {angularLimitsMax=<1 0 15>, angularLimitsMin=<-1 -30 -15>, angularTarget=<0 0 0>}
      localJointOffset=<2 0 0>
     }]
    relativeSpaceBone {boneName=dyn_root_simspace}
    sphericalLimits [{limitRadius=7.5, drivingBone {boneName=spine_03}, sphereLocalOffset=<0 2 0>}]
   }
   n80 = Anim Dynamics - Bone: dyn_jacket_d_02(ComponentPose=@n79, Alpha=@n76, GravityOverride=@n70, ExternalForce="0,0,0") {
    angularDampingOverride=0.9999
    angularSpringConstant=50
    bAngularSpring=true
    bGravityOverrideInSimSpace=true
    bOverrideAngularDamping=true
    bUseGravityOverride=true
    bUseSphericalLimits=true
    simulationSpace=BoneRelative
    alphaBoolBlend {blendOption=Linear}
    boundBone {boneName=dyn_jacket_d_02}
    componentAppliedLinearAccClamp=<3500 3500 3500>
    componentLinearAccScale=<0.8 0.8 0.8>
    componentLinearVelScale=<0.8 0.8 0.8>
    physicsBodyDefinitions [{
      collisionType=InnerSphere
      sphereCollisionRadius=1
      boundBone {boneName=dyn_jacket_d_02}
      boxExtents=<10 20 20>
      constraintSetup {angularLimitsMax=<1 0 15>, angularLimitsMin=<-1 -30 -15>, angularTarget=<0 0 0>}
      localJointOffset=<2 0 0>
     }]
    relativeSpaceBone {boneName=dyn_root_simspace}
    sphericalLimits [{limitRadius=7, drivingBone {boneName=spine_01}, sphereLocalOffset=<3 2 0>}]
   }
   n81 = Anim Dynamics - Bone: dyn_jacket_e_01(ComponentPose=@n80, Alpha=@n76, GravityOverride=@n70, ExternalForce="0,0,0") {
    angularDampingOverride=0.9999
    angularSpringConstant=50
    bAngularSpring=true
    bGravityOverrideInSimSpace=true
    bOverrideAngularDamping=true
    bUseGravityOverride=true
    bUseSphericalLimits=true
    simulationSpace=BoneRelative
    alphaBoolBlend {blendOption=Linear}
    boundBone {boneName=dyn_jacket_e_01}
    componentAppliedLinearAccClamp=<3500 3500 3500>
    componentLinearAccScale=<0.8 0.8 0.8>
    componentLinearVelScale=<0.8 0.8 0.8>
    physicsBodyDefinitions [{
      collisionType=InnerSphere
      sphereCollisionRadius=1
      boundBone {boneName=dyn_jacket_e_01}
      boxExtents=<10 20 20>
      constraintSetup {angularLimitsMax=<1 0 15>, angularLimitsMin=<-1 -30 -15>, angularTarget=<0 0 0>}
      localJointOffset=<2 0 0>
     }]
    relativeSpaceBone {boneName=dyn_root_simspace}
    sphericalLimits [{limitRadius=7.5, drivingBone {boneName=spine_03}, sphereLocalOffset=<0 0 1>}]
   }
   n82 = Anim Dynamics - Bone: dyn_jacket_e_02(ComponentPose=@n81, Alpha=@n76, GravityOverride=@n70, ExternalForce="0,0,0") {
    angularDampingOverride=0.9999
    angularSpringConstant=50
    bAngularSpring=true
    bGravityOverrideInSimSpace=true
    bOverrideAngularDamping=true
    bUseGravityOverride=true
    bUseSphericalLimits=true
    simulationSpace=BoneRelative
    alphaBoolBlend {blendOption=Linear}
    boundBone {boneName=dyn_jacket_e_02}
    componentAppliedLinearAccClamp=<3500 3500 3500>
    componentLinearAccScale=<0.8 0.8 0.8>
    componentLinearVelScale=<0.8 0.8 0.8>
    physicsBodyDefinitions [{
      collisionType=InnerSphere
      sphereCollisionRadius=1
      boundBone {boneName=dyn_jacket_e_02}
      boxExtents=<10 20 20>
      constraintSetup {angularLimitsMax=<1 0 15>, angularLimitsMin=<-1 -30 -15>, angularTarget=<0 0 0>}
      localJointOffset=<2 0 0>
     }]
    relativeSpaceBone {boneName=dyn_root_simspace}
    sphericalLimits [{limitRadius=10, drivingBone {boneName=spine_01}, sphereLocalOffset=<0 3.5 0>}]
   }
   n83 = Anim Dynamics - Bone: dyn_jacket_f_01(ComponentPose=@n82, Alpha=1, GravityOverride="0,0,0", ExternalForce="0,0,0") {
    angularDampingOverride=0.9999
    angularSpringConstant=50
    bAngularSpring=true
    bGravityOverrideInSimSpace=true
    bOverrideAngularDamping=true
    simulationSpace=BoneRelative
    alphaBoolBlend {blendOption=Linear}
    boundBone {boneName=dyn_jacket_f_01}
    componentAppliedLinearAccClamp=<3500 3500 3500>
    componentLinearAccScale=<0.6 0.6 0.6>
    componentLinearVelScale=<0.6 0.6 0.6>
    physicsBodyDefinitions [{
      sphereCollisionRadius=0
      boundBone {boneName=dyn_jacket_f_01}
      boxExtents=<10 20 20>
      constraintSetup {angularLimitsMax=<1 0 5>, angularLimitsMin=<-1 -30 -5>, angularTarget=<0 0 0>}
      localJointOffset=<2 0 0>
     }]
    relativeSpaceBone {boneName=dyn_root_simspace}
   }
   n84 = Anim Dynamics - Bone: dyn_jacket_f_02(ComponentPose=@n83, Alpha=1, GravityOverride="0,0,0", ExternalForce="0,0,0") {
    angularDampingOverride=0.9999
    angularSpringConstant=50
    bAngularSpring=true
    bGravityOverrideInSimSpace=true
    bOverrideAngularDamping=true
    bUseSphericalLimits=true
    simulationSpace=BoneRelative
    alphaBoolBlend {blendOption=Linear}
    boundBone {boneName=dyn_jacket_f_02}
    componentAppliedLinearAccClamp=<3500 3500 3500>
    componentLinearAccScale=<0.6 0.6 0.6>
    componentLinearVelScale=<0.6 0.6 0.6>
    physicsBodyDefinitions [{
      collisionType=InnerSphere
      sphereCollisionRadius=1
      boundBone {boneName=dyn_jacket_f_02}
      boxExtents=<10 20 20>
      constraintSetup {angularLimitsMax=<1 0 5>, angularLimitsMin=<-1 -30 -5>, angularTarget=<0 0 0>}
      localJointOffset=<2 0 0>
     }]
    relativeSpaceBone {boneName=dyn_root_simspace}
    sphericalLimits [
     {limitRadius=9, drivingBone {boneName=spine_01}, sphereLocalOffset=<0 0 1.5>}
     {limitRadius=12, drivingBone {boneName=pelvis}, sphereLocalOffset=<0 0 0.5>}
    ]
   }
   n85 = Anim Dynamics - Bone: dyn_jacket_g_01(ComponentPose=@n84, Alpha=1, GravityOverride="0,0,0", ExternalForce="0,0,0") {
    angularDampingOverride=0.9999
    angularSpringConstant=50
    bAngularSpring=true
    bGravityOverrideInSimSpace=true
    bOverrideAngularDamping=true
    simulationSpace=BoneRelative
    alphaBoolBlend {blendOption=Linear}
    boundBone {boneName=dyn_jacket_g_01}
    componentAppliedLinearAccClamp=<3500 3500 3500>
    componentLinearAccScale=<0.6 0.6 0.6>
    componentLinearVelScale=<0.6 0.6 0.6>
    physicsBodyDefinitions [{
      sphereCollisionRadius=0
      boundBone {boneName=dyn_jacket_g_01}
      boxExtents=<10 20 20>
      constraintSetup {angularLimitsMax=<1 -2 5>, angularLimitsMin=<-1 -15 -5>, angularTarget=<0 0 0>}
      localJointOffset=<2 0 0>
     }]
    relativeSpaceBone {boneName=dyn_root_simspace}
   }
   n86 = Anim Dynamics - Bone: dyn_jacket_g_02(ComponentPose=@n85, Alpha=1, GravityOverride="0,0,0", ExternalForce="0,0,0") {
    angularDampingOverride=0.9999
    angularSpringConstant=50
    bAngularSpring=true
    bGravityOverrideInSimSpace=true
    bOverrideAngularDamping=true
    bUseSphericalLimits=true
    simulationSpace=BoneRelative
    alphaBoolBlend {blendOption=Linear}
    boundBone {boneName=dyn_jacket_g_02}
    componentAppliedLinearAccClamp=<3500 3500 3500>
    componentLinearAccScale=<0.6 0.6 0.6>
    componentLinearVelScale=<0.6 0.6 0.6>
    physicsBodyDefinitions [{
      collisionType=CustomSphere
      sphereCollisionRadius=1
      boundBone {boneName=dyn_jacket_g_02}
      boxExtents=<10 20 20>
      constraintSetup {angularLimitsMax=<1 -2 5>, angularLimitsMin=<-1 -15 -5>, angularTarget=<0 0 0>}
      localJointOffset=<2 0 0>
     }]
    relativeSpaceBone {boneName=dyn_root_simspace}
   }
   n87 = Anim Dynamics - Bone: dyn_jacket_g_03(ComponentPose=@n86, Alpha=1, GravityOverride="0,0,0", ExternalForce="0,0,0") {
    angularDampingOverride=0.9999
    angularSpringConstant=50
    bAngularSpring=true
    bGravityOverrideInSimSpace=true
    bOverrideAngularDamping=true
    bUseSphericalLimits=true
    simulationSpace=BoneRelative
    alphaBoolBlend {blendOption=Linear}
    boundBone {boneName=dyn_jacket_g_03}
    componentAppliedLinearAccClamp=<3500 3500 3500>
    componentLinearAccScale=<0.6 0.6 0.6>
    componentLinearVelScale=<0.6 0.6 0.6>
    physicsBodyDefinitions [{
      collisionType=InnerSphere
      sphereCollisionRadius=1
      boundBone {boneName=dyn_jacket_g_03}
      boxExtents=<10 20 20>
      constraintSetup {angularLimitsMax=<1 -2 5>, angularLimitsMin=<-1 -15 -5>, angularTarget=<0 0 0>}
      localJointOffset=<2 0 0>
     }]
    relativeSpaceBone {boneName=dyn_root_simspace}
    sphericalLimits [
     {limitRadius=9.5, drivingBone {boneName=spine_01}, sphereLocalOffset=<0 0 0>}
     {limitRadius=10, drivingBone {boneName=spine_02}, sphereLocalOffset=<0 0 0>}
     {limitRadius=11, drivingBone {boneName=pelvis}, sphereLocalOffset=<0 0 0>}
    ]
   }
   n88 = Component To Local(ComponentPose=@n87)
   n89 = Get bIsInFrontEnd()
   n90 = "Transform (Modify) Bone - Bone: dyn_jacket_c_MSTR"!(
    ComponentPose=Local To Component(LocalPose=Blend Poses by bool(BlendPose_0=@n68, BlendPose_1=@n88, bActiveValue=@n89) {blendType=Linear})
    bAlphaBoolEnabled:bool=True
    Translation="0,0,5"
    Rotation="0,0,0"
    Scale="1,1,1"
   ) -> (Pose) {
    alphaInputType=Bool
    rotationMode=BMM_Additive
    rotationSpace=BCS_BoneSpace
    scaleMode=BMM_Additive
    scaleSpace=BCS_BoneSpace
    translationMode=BMM_Additive
    translationSpace=BCS_BoneSpace
    alphaBoolBlend {blendInTime=0.1, blendOption=Linear, blendOutTime=0.1}
    boneToModify {boneName=dyn_jacket_c_MSTR}
    translation {z=5}
   }
   n91 = "Transform (Modify) Bone - Bone: dyn_jacket_d_MSTR"!(ComponentPose=@n90, bAlphaBoolEnabled:bool=True, Translation="0,0,5", Rotation="0,0,0", Scale="1,1,1") -> (Pose) {
    alphaInputType=Bool
    rotationMode=BMM_Additive
    rotationSpace=BCS_BoneSpace
    scaleMode=BMM_Additive
    scaleSpace=BCS_BoneSpace
    translationMode=BMM_Additive
    translationSpace=BCS_BoneSpace
    alphaBoolBlend {blendInTime=0.1, blendOption=Linear, blendOutTime=0.1}
    boneToModify {boneName=dyn_jacket_d_MSTR}
    translation {z=5}
   }
   n92 = "Transform (Modify) Bone - Bone: dyn_jacket_e_MSTR"!(ComponentPose=@n91, bAlphaBoolEnabled:bool=True, Translation="0,0,5", Rotation:Rotator="0,0,0", Scale:Vector="1,1,1") -> (Pose) {
    alphaInputType=Bool
    rotationMode=BMM_Additive
    rotationSpace=BCS_BoneSpace
    scaleMode=BMM_Additive
    scaleSpace=BCS_BoneSpace
    translationMode=BMM_Additive
    translationSpace=BCS_BoneSpace
    alphaBoolBlend {blendInTime=0.1, blendOption=Linear, blendOutTime=0.1}
    boneToModify {boneName=dyn_jacket_e_MSTR}
    translation {z=5}
   }
   n93 = Get crouch alpha() -> (crouch alpha:double)
   n94 = "Constraint - dyn_jacket_a_MSTR (P)"(ComponentPose=@n92, Alpha=@n93, ConstraintWeights_0=1) {
    alphaBoolBlend {blendOption=Linear}
    boneToModify {boneName=dyn_jacket_a_MSTR}
    constraintSetup [{targetBone {boneName=pelvis}}]
    constraintWeights [1]
   }
   n95 = "Constraint - dyn_jacket_a_01 (P)"(ComponentPose=@n94, Alpha=@n93, ConstraintWeights_0=1) {
    alphaBoolBlend {blendOption=Linear}
    boneToModify {boneName=dyn_jacket_a_01}
    constraintSetup [{targetBone {boneName=pelvis}}]
    constraintWeights [1]
   }
   n96 = "Constraint - dyn_jacket_a_02 (P)"(ComponentPose=@n95, Alpha=@n93, ConstraintWeights_0=1) {
    alphaBoolBlend {blendOption=Linear}
    boneToModify {boneName=dyn_jacket_a_02}
    constraintSetup [{targetBone {boneName=pelvis}}]
    constraintWeights [1]
   }
   n97 = "Constraint - dyn_jacket_a_03 (P)"(ComponentPose=@n96, Alpha=@n93, ConstraintWeights_0=1) {
    alphaBoolBlend {blendOption=Linear}
    boneToModify {boneName=dyn_jacket_a_03}
    constraintSetup [{targetBone {boneName=pelvis}}]
    constraintWeights [1]
   }
   n98 = "Transform (Modify) Bone - Bone: dyn_jacket_a_MSTR"!(ComponentPose=@n97, bAlphaBoolEnabled:bool=True, Translation="3,2,0", Rotation="10.586,-40.335,-25", Scale="1,1,1") -> (Pose) {
    alphaInputType=Bool
    rotationMode=BMM_Additive
    scaleMode=BMM_Additive
    translationMode=BMM_Additive
    alphaBoolBlend {blendInTime=0.2, blendOption=Linear, blendOutTime=0.2}
    boneToModify {boneName=dyn_jacket_a_MSTR}
    rotation=<rot 10.586 -40.335 -25>
    translation=<3 2>
   }
   n99 = "Transform (Modify) Bone - Bone: dyn_jacket_a_02"!(ComponentPose=@n98, bAlphaBoolEnabled:bool=True, Translation="0,0,0", Rotation="0,0,0", Scale="1,1,1") -> (Pose) {
    alphaInputType=Bool
    rotationMode=BMM_Additive
    scaleMode=BMM_Additive
    translationMode=BMM_Additive
    alphaBoolBlend {blendInTime=0.1, blendOption=Linear, blendOutTime=0.1}
    boneToModify {boneName=dyn_jacket_a_02}
   }
   n100 = "Transform (Modify) Bone - Bone: dyn_jacket_a_03"!(ComponentPose=@n99, bAlphaBoolEnabled:bool=True, Translation="0,0,0", Rotation="0,0,0", Scale="1,1,1") -> (Pose) {
    alphaInputType=Bool
    rotationMode=BMM_Additive
    scaleMode=BMM_Additive
    translationMode=BMM_Additive
    alphaBoolBlend {blendInTime=0.1, blendOption=Linear, blendOutTime=0.1}
    boneToModify {boneName=dyn_jacket_a_03}
   }
   n101 = "Transform (Modify) Bone - Bone: dyn_jacket_a_02"!(ComponentPose=@n100, Alpha:float=Get LookDown() -> (LookDown:double), Translation="0,0,0", Rotation="20,0,0", Scale="1,1,1") -> (Pose) {
    rotationMode=BMM_Additive
    scaleMode=BMM_Additive
    translationMode=BMM_Additive
    alphaBoolBlend {blendInTime=0.1, blendOption=Linear, blendOutTime=0.1}
    boneToModify {boneName=dyn_jacket_a_02}
    rotation {pitch=20}
   }
   n102 = "Constraint - dyn_zipper_MSTR (PPPPPPPPPP)"!(
    ComponentPose=@n101
    Alpha=1
    ConstraintWeights_0=0.0043
    ConstraintWeights_1:float=0.0014
    ConstraintWeights_2:float=0.0797
    ConstraintWeights_3:float=0.0193
    ConstraintWeights_4:float=0.0299
    ConstraintWeights_5:float=0.27
    ConstraintWeights_6:float=0.5488
    ConstraintWeights_7:float=0.0155
    ConstraintWeights_8:float=0.03
    ConstraintWeights_9:float=0.0009
   ) -> (Pose) {
    alphaBoolBlend {blendOption=Linear}
    boneToModify {boneName=dyn_zipper_MSTR}
    constraintSetup [
     {targetBone {boneName=dyn_jacket_e_02}}
     {targetBone {boneName=dyn_jacket_e_01}}
     {targetBone {boneName=dyn_jacket_f_02}}
     {targetBone {boneName=dyn_jacket_f_01}}
     {targetBone {boneName=dyn_jacket_g_01}}
     {targetBone {boneName=dyn_jacket_g_02}}
     {targetBone {boneName=dyn_jacket_g_03}}
     {targetBone {boneName=spine_03}}
     {targetBone {boneName=spine_04}}
     {targetBone {boneName=spine_05}}
    ]
    constraintWeights [1,1,1,1,1,1,1,1,1,1]
   }
   n103 = Anim Dynamics - Bone: dyn_zipper(ComponentPose=@n102, Alpha=1) {
    angularDampingOverride=0.9999
    angularSpringConstant=25
    bAngularSpring=true
    bGravityOverrideInSimSpace=true
    bOverrideAngularDamping=true
    simulationSpace=BoneRelative
    alphaBoolBlend {blendOption=Linear}
    boundBone {boneName=dyn_zipper}
    componentAppliedLinearAccClamp=<3500 3500 3500>
    componentLinearAccScale=<1.7 1.7 1.7>
    componentLinearVelScale=<1.7 1.7 1.7>
    physicsBodyDefinitions [{
      sphereCollisionRadius=0
      boundBone {boneName=dyn_zipper}
      boxExtents=<10 10 5>
      constraintSetup {angularLimitsMax=<1 2 15>, angularLimitsMin=<-1 -30 -15>, angularTarget=<0 0 0>}
      localJointOffset=<2 0 0>
     }]
    relativeSpaceBone {boneName=dyn_root_simspace}
   }
   n104 = "Constraint - dyn_tail_MSTR (PPPPPP)"!(
    ComponentPose=@n103
    Alpha=1
    ConstraintWeights_0=0.0454
    ConstraintWeights_1:float=0.0541
    ConstraintWeights_2:float=0
    ConstraintWeights_3:float=0.0025
    ConstraintWeights_4:float=0.8525
    ConstraintWeights_5:float=0.0455
   ) -> (Pose) {
    alphaBoolBlend {blendOption=Linear}
    boneToModify {boneName=dyn_tail_MSTR}
    constraintSetup [
     {targetBone {boneName=thigh_twist_01_r}}
     {targetBone {boneName=spine_01}}
     {targetBone {boneName=spine_02}}
     {targetBone {boneName=spine_03}}
     {targetBone {boneName=pelvis}}
     {targetBone {boneName=thigh_twist_01_l}}
    ]
    constraintWeights [1,1,1,1,1,1]
   }
   Save cached pose 'prerigid'(Pose=Component To Local(ComponentPose=@n104)) {cacheName=prerigid}
   n105 = Layered blend per bone(
    BasePose=Copy Pose From Mesh() {bCopyCurves=true, bUseAttachedParent=true}
    BlendPoses_0="Sequence Evaluator (by time) 'Emote_CerealBox_CMM_Hooves'"!(ExplicitTime:float=Get emoteOverwriteJoint_Position() -> (emoteOverwriteJoint_Position:double)) -> (Pose:Pose Link Structure) {is=title AnimGraphNode_SequenceEvaluator, sequence=/BRCosmetics/Animation/Game/MainPlayer/Emotes/CerealBox/Emote_CerealBox_CMM_Hooves}
    BlendWeights_0=Get EmoteAlpha() -> (EmoteAlpha:double)
   ) {layerSetup [{branchFilters [{boneName=hoof_c_l}, {boneName=hoof_c_r}, {boneName=dyn_hoof_hand_r}, {boneName=dyn_hoof_hand_l}]}]}
   Save cached pose 'EmoteBlend_pose'(Pose=@n105) {cacheName=EmoteBlend_pose}
   n106 = "Transform (Modify) Bone - Bone: dyn_thumb_01_r"!(
    ComponentPose=Local To Component(LocalPose=Use cached pose 'EmoteBlend_pose'())
    bAlphaBoolEnabled:bool=Get finger scale() -> (finger scale:bool)
    Translation="0,0,0"
    Rotation="0,0,0"
    Scale="0.5,0.5,0.5"
   ) -> (Pose) {
    alphaInputType=Bool
    scaleMode=BMM_Additive
    scaleSpace=BCS_BoneSpace
    alphaBoolBlend {blendOption=Linear}
    boneToModify {boneName=dyn_thumb_01_r}
    scale=<0.5 0.5 0.5>
   }
   n107 = "Constraint - dyn_hoof_hand_l (R)"(ComponentPose=@n106, Alpha=1, ConstraintWeights_0=1) {
    alphaBoolBlend {blendOption=Linear}
    boneToModify {boneName=dyn_hoof_hand_l}
    constraintSetup [{transformType=Rotation, targetBone {boneName=hand_l}}]
    constraintWeights [1]
   }
   n108 = "Constraint - dyn_hoof_hand_r (R)"(ComponentPose=@n107, Alpha=1, ConstraintWeights_0=1) {
    alphaBoolBlend {blendOption=Linear}
    boneToModify {boneName=dyn_hoof_hand_r}
    constraintSetup [{transformType=Rotation, targetBone {boneName=hand_r}}]
    constraintWeights [1]
   }
   n109 = "Constraint - dyn_hand_l (R)"(ComponentPose=@n108, Alpha=1, ConstraintWeights_0=1) {
    alphaBoolBlend {blendOption=Linear}
    boneToModify {boneName=dyn_hand_l}
    constraintSetup [{transformType=Rotation, targetBone {boneName=hand_l}}]
    constraintWeights [1]
   }
   n110 = "Constraint - dyn_hand_r (R)"(ComponentPose=@n109, Alpha=1, ConstraintWeights_0=1) {
    alphaBoolBlend {blendOption=Linear}
    boneToModify {boneName=dyn_hand_r}
    constraintSetup [{transformType=Rotation, targetBone {boneName=hand_r}}]
    constraintWeights [1]
   }
   n111 = "Constraint - dyn_thumb_01_l (R)"(ComponentPose=@n110, Alpha=1, ConstraintWeights_0=1) {
    alphaBoolBlend {blendOption=Linear}
    boneToModify {boneName=dyn_thumb_01_l}
    constraintSetup [{transformType=Rotation, targetBone {boneName=thumb_01_l}}]
    constraintWeights [1]
   }
   n112 = "Constraint - dyn_thumb_02_l (R)"(ComponentPose=@n111, Alpha=1, ConstraintWeights_0=1) {
    alphaBoolBlend {blendOption=Linear}
    boneToModify {boneName=dyn_thumb_02_l}
    constraintSetup [{transformType=Rotation, targetBone {boneName=thumb_02_l}}]
    constraintWeights [1]
   }
   n113 = "Constraint - dyn_thumb_03_l (R)"(ComponentPose=@n112, Alpha=1, ConstraintWeights_0=1) {
    alphaBoolBlend {blendOption=Linear}
    boneToModify {boneName=dyn_thumb_03_l}
    constraintSetup [{transformType=Rotation, targetBone {boneName=thumb_03_l}}]
    constraintWeights [1]
   }
   n114 = "Constraint - dyn_index_metacarpal_l (R)"(ComponentPose=@n113, Alpha=1, ConstraintWeights_0=1) {
    alphaBoolBlend {blendOption=Linear}
    boneToModify {boneName=dyn_index_metacarpal_l}
    constraintSetup [{transformType=Rotation, targetBone {boneName=index_metacarpal_l}}]
    constraintWeights [1]
   }
   n115 = "Constraint - dyn_index_01_l (R)"(ComponentPose=@n114, Alpha=1, ConstraintWeights_0=1) {
    alphaBoolBlend {blendOption=Linear}
    boneToModify {boneName=dyn_index_01_l}
    constraintSetup [{transformType=Rotation, targetBone {boneName=index_01_l}}]
    constraintWeights [1]
   }
   n116 = "Constraint - dyn_index_02_l (R)"(ComponentPose=@n115, Alpha=1, ConstraintWeights_0=1) {
    alphaBoolBlend {blendOption=Linear}
    boneToModify {boneName=dyn_index_02_l}
    constraintSetup [{transformType=Rotation, targetBone {boneName=index_02_l}}]
    constraintWeights [1]
   }
   n117 = "Constraint - dyn_index_03_l (R)"(ComponentPose=@n116, Alpha=1, ConstraintWeights_0=1) {
    alphaBoolBlend {blendOption=Linear}
    boneToModify {boneName=dyn_index_03_l}
    constraintSetup [{transformType=Rotation, targetBone {boneName=index_03_l}}]
    constraintWeights [1]
   }
   n118 = "Constraint - dyn_middle_metacarpal_l (R)"(ComponentPose=@n117, Alpha=1, ConstraintWeights_0=1) {
    alphaBoolBlend {blendOption=Linear}
    boneToModify {boneName=dyn_middle_metacarpal_l}
    constraintSetup [{transformType=Rotation, targetBone {boneName=middle_metacarpal_l}}]
    constraintWeights [1]
   }
   n119 = "Constraint - dyn_middle_01_l (R)"(ComponentPose=@n118, Alpha=1, ConstraintWeights_0=1) {
    alphaBoolBlend {blendOption=Linear}
    boneToModify {boneName=dyn_middle_01_l}
    constraintSetup [{transformType=Rotation, targetBone {boneName=middle_01_l}}]
    constraintWeights [1]
   }
   n120 = "Constraint - dyn_middle_02_l (R)"(ComponentPose=@n119, Alpha=1, ConstraintWeights_0=1) {
    alphaBoolBlend {blendOption=Linear}
    boneToModify {boneName=dyn_middle_02_l}
    constraintSetup [{transformType=Rotation, targetBone {boneName=middle_02_l}}]
    constraintWeights [1]
   }
   n121 = "Constraint - dyn_middle_03_l (R)"(ComponentPose=@n120, Alpha=1, ConstraintWeights_0=1) {
    alphaBoolBlend {blendOption=Linear}
    boneToModify {boneName=dyn_middle_03_l}
    constraintSetup [{transformType=Rotation, targetBone {boneName=middle_03_l}}]
    constraintWeights [1]
   }
   n122 = "Constraint - dyn_ring_metacarpal_l (R)"(ComponentPose=@n121, Alpha=1, ConstraintWeights_0=1) {
    alphaBoolBlend {blendOption=Linear}
    boneToModify {boneName=dyn_ring_metacarpal_l}
    constraintSetup [{transformType=Rotation, targetBone {boneName=ring_metacarpal_l}}]
    constraintWeights [1]
   }
   n123 = "Constraint - dyn_ring_01_l (R)"(ComponentPose=@n122, Alpha=1, ConstraintWeights_0=1) {
    alphaBoolBlend {blendOption=Linear}
    boneToModify {boneName=dyn_ring_01_l}
    constraintSetup [{transformType=Rotation, targetBone {boneName=ring_01_l}}]
    constraintWeights [1]
   }
   n124 = "Constraint - dyn_ring_02_l (R)"(ComponentPose=@n123, Alpha=1, ConstraintWeights_0=1) {
    alphaBoolBlend {blendOption=Linear}
    boneToModify {boneName=dyn_ring_02_l}
    constraintSetup [{transformType=Rotation, targetBone {boneName=ring_02_l}}]
    constraintWeights [1]
   }
   n125 = "Constraint - dyn_ring_03_l (R)"(ComponentPose=@n124, Alpha=1, ConstraintWeights_0=1) {
    alphaBoolBlend {blendOption=Linear}
    boneToModify {boneName=dyn_ring_03_l}
    constraintSetup [{transformType=Rotation, targetBone {boneName=ring_03_l}}]
    constraintWeights [1]
   }
   n126 = "Constraint - dyn_pinky_metacarpal_l (R)"(ComponentPose=@n125, Alpha=1, ConstraintWeights_0=1) {
    alphaBoolBlend {blendOption=Linear}
    boneToModify {boneName=dyn_pinky_metacarpal_l}
    constraintSetup [{transformType=Rotation, targetBone {boneName=pinky_metacarpal_l}}]
    constraintWeights [1]
   }
   n127 = "Constraint - dyn_pinky_01_l (R)"(ComponentPose=@n126, Alpha=1, ConstraintWeights_0=1) {
    alphaBoolBlend {blendOption=Linear}
    boneToModify {boneName=dyn_pinky_01_l}
    constraintSetup [{transformType=Rotation, targetBone {boneName=pinky_01_l}}]
    constraintWeights [1]
   }
   n128 = "Constraint - dyn_pinky_02_l (R)"(ComponentPose=@n127, Alpha=1, ConstraintWeights_0=1) {
    alphaBoolBlend {blendOption=Linear}
    boneToModify {boneName=dyn_pinky_02_l}
    constraintSetup [{transformType=Rotation, targetBone {boneName=pinky_02_l}}]
    constraintWeights [1]
   }
   n129 = "Constraint - dyn_pinky_03_l (R)"(ComponentPose=@n128, Alpha=1, ConstraintWeights_0=1) {
    alphaBoolBlend {blendOption=Linear}
    boneToModify {boneName=dyn_pinky_03_l}
    constraintSetup [{transformType=Rotation, targetBone {boneName=pinky_03_l}}]
    constraintWeights [1]
   }
   n130 = "Constraint - dyn_thumb_01_r (R)"(ComponentPose=@n129, Alpha=1, ConstraintWeights_0=1) {
    alphaBoolBlend {blendOption=Linear}
    boneToModify {boneName=dyn_thumb_01_r}
    constraintSetup [{transformType=Rotation, targetBone {boneName=thumb_01_r}}]
    constraintWeights [1]
   }
   n131 = "Constraint - dyn_thumb_02_r (R)"(ComponentPose=@n130, Alpha=1, ConstraintWeights_0=1) {
    alphaBoolBlend {blendOption=Linear}
    boneToModify {boneName=dyn_thumb_02_r}
    constraintSetup [{transformType=Rotation, targetBone {boneName=thumb_02_r}}]
    constraintWeights [1]
   }
   n132 = "Constraint - dyn_thumb_03_r (R)"(ComponentPose=@n131, Alpha=1, ConstraintWeights_0=1) {
    alphaBoolBlend {blendOption=Linear}
    boneToModify {boneName=dyn_thumb_03_r}
    constraintSetup [{transformType=Rotation, targetBone {boneName=thumb_03_r}}]
    constraintWeights [1]
   }
   n133 = "Constraint - dyn_index_metacarpal_r (R)"(ComponentPose=@n132, Alpha=1, ConstraintWeights_0=1) {
    alphaBoolBlend {blendOption=Linear}
    boneToModify {boneName=dyn_index_metacarpal_r}
    constraintSetup [{transformType=Rotation, targetBone {boneName=index_metacarpal_r}}]
    constraintWeights [1]
   }
   n134 = "Constraint - dyn_index_01_r (R)"(ComponentPose=@n133, Alpha=1, ConstraintWeights_0=1) {
    alphaBoolBlend {blendOption=Linear}
    boneToModify {boneName=dyn_index_01_r}
    constraintSetup [{transformType=Rotation, targetBone {boneName=index_01_r}}]
    constraintWeights [1]
   }
   n135 = "Constraint - dyn_index_02_r (R)"(ComponentPose=@n134, Alpha=1, ConstraintWeights_0=1) {
    alphaBoolBlend {blendOption=Linear}
    boneToModify {boneName=dyn_index_02_r}
    constraintSetup [{transformType=Rotation, targetBone {boneName=index_02_r}}]
    constraintWeights [1]
   }
   n136 = "Constraint - dyn_index_03_r (R)"(ComponentPose=@n135, Alpha=1, ConstraintWeights_0=1) {
    alphaBoolBlend {blendOption=Linear}
    boneToModify {boneName=dyn_index_03_r}
    constraintSetup [{transformType=Rotation, targetBone {boneName=index_03_r}}]
    constraintWeights [1]
   }
   n137 = "Constraint - dyn_middle_metacarpal_r (R)"(ComponentPose=@n136, Alpha=1, ConstraintWeights_0=1) {
    alphaBoolBlend {blendOption=Linear}
    boneToModify {boneName=dyn_middle_metacarpal_r}
    constraintSetup [{transformType=Rotation, targetBone {boneName=middle_metacarpal_r}}]
    constraintWeights [1]
   }
   n138 = "Constraint - dyn_middle_01_r (R)"(ComponentPose=@n137, Alpha=1, ConstraintWeights_0=1) {
    alphaBoolBlend {blendOption=Linear}
    boneToModify {boneName=dyn_middle_01_r}
    constraintSetup [{transformType=Rotation, targetBone {boneName=middle_01_r}}]
    constraintWeights [1]
   }
   n139 = "Constraint - dyn_middle_02_r (R)"(ComponentPose=@n138, Alpha=1, ConstraintWeights_0=1) {
    alphaBoolBlend {blendOption=Linear}
    boneToModify {boneName=dyn_middle_02_r}
    constraintSetup [{transformType=Rotation, targetBone {boneName=middle_02_r}}]
    constraintWeights [1]
   }
   n140 = "Constraint - dyn_middle_03_r (R)"(ComponentPose=@n139, Alpha=1, ConstraintWeights_0=1) {
    alphaBoolBlend {blendOption=Linear}
    boneToModify {boneName=dyn_middle_03_r}
    constraintSetup [{transformType=Rotation, targetBone {boneName=middle_03_r}}]
    constraintWeights [1]
   }
   n141 = "Constraint - dyn_ring_metacarpal_r (R)"(ComponentPose=@n140, Alpha=1, ConstraintWeights_0=1) {
    alphaBoolBlend {blendOption=Linear}
    boneToModify {boneName=dyn_ring_metacarpal_r}
    constraintSetup [{transformType=Rotation, targetBone {boneName=ring_metacarpal_r}}]
    constraintWeights [1]
   }
   n142 = "Constraint - dyn_ring_01_r (R)"(ComponentPose=@n141, Alpha=1, ConstraintWeights_0=1) {
    alphaBoolBlend {blendOption=Linear}
    boneToModify {boneName=dyn_ring_01_r}
    constraintSetup [{transformType=Rotation, targetBone {boneName=ring_01_r}}]
    constraintWeights [1]
   }
   n143 = "Constraint - dyn_ring_02_r (R)"(ComponentPose=@n142, Alpha=1, ConstraintWeights_0=1) {
    alphaBoolBlend {blendOption=Linear}
    boneToModify {boneName=dyn_ring_02_r}
    constraintSetup [{transformType=Rotation, targetBone {boneName=ring_02_r}}]
    constraintWeights [1]
   }
   n144 = "Constraint - dyn_ring_03_r (R)"(ComponentPose=@n143, Alpha=1, ConstraintWeights_0=1) {
    alphaBoolBlend {blendOption=Linear}
    boneToModify {boneName=dyn_ring_03_r}
    constraintSetup [{transformType=Rotation, targetBone {boneName=dyn_ring_03_r}}]
    constraintWeights [1]
   }
   n145 = "Constraint - dyn_pinky_metacarpal_r (R)"(ComponentPose=@n144, Alpha=1, ConstraintWeights_0=1) {
    alphaBoolBlend {blendOption=Linear}
    boneToModify {boneName=dyn_pinky_metacarpal_r}
    constraintSetup [{transformType=Rotation, targetBone {boneName=pinky_metacarpal_r}}]
    constraintWeights [1]
   }
   n146 = "Constraint - dyn_pinky_01_r (R)"(ComponentPose=@n145, Alpha=1, ConstraintWeights_0=1) {
    alphaBoolBlend {blendOption=Linear}
    boneToModify {boneName=dyn_pinky_01_r}
    constraintSetup [{transformType=Rotation, targetBone {boneName=pinky_01_r}}]
    constraintWeights [1]
   }
   n147 = "Constraint - dyn_pinky_02_r (R)"(ComponentPose=@n146, Alpha=1, ConstraintWeights_0=1) {
    alphaBoolBlend {blendOption=Linear}
    boneToModify {boneName=dyn_pinky_02_r}
    constraintSetup [{transformType=Rotation, targetBone {boneName=pinky_02_r}}]
    constraintWeights [1]
   }
   n148 = "Constraint - dyn_pinky_03_r (R)"(ComponentPose=@n147, Alpha=1, ConstraintWeights_0=1) {
    alphaBoolBlend {blendOption=Linear}
    boneToModify {boneName=dyn_pinky_03_r}
    constraintSetup [{transformType=Rotation, targetBone {boneName=pinky_03_r}}]
    constraintWeights [1]
   }
   n149 = "Constraint - dyn_jacket_a_constraint (PPPPPP)"!(
    ComponentPose=@n148
    Alpha=1
    ConstraintWeights_0=0.0061
    ConstraintWeights_1:float=0.0025
    ConstraintWeights_2:float=0.0545
    ConstraintWeights_3:float=0.1825
    ConstraintWeights_4:float=0.5306
    ConstraintWeights_5:float=0.2238
   ) -> (Pose) {
    alphaBoolBlend {blendOption=Linear}
    boneToModify {boneName=dyn_jacket_a_constraint}
    constraintSetup [
     {targetBone {boneName=clavicle_l}}
     {targetBone {boneName=spine_01}}
     {targetBone {boneName=spine_02}}
     {targetBone {boneName=spine_03}}
     {targetBone {boneName=spine_04}}
     {targetBone {boneName=spine_05}}
    ]
    constraintWeights [1,1,1,1,1,1]
   }
   n150 = "Constraint - dyn_jacket_b_constraint (PPPPP)"!(
    ComponentPose=@n149
    Alpha=1
    ConstraintWeights_0=0.1178
    ConstraintWeights_1:float=0.3472
    ConstraintWeights_2:float=0.4759
    ConstraintWeights_3:float=0.028
    ConstraintWeights_4:float=0.031
   ) -> (Pose) {
    alphaBoolBlend {blendOption=Linear}
    boneToModify {boneName=dyn_jacket_b_constraint}
    constraintSetup [
     {targetBone {boneName=spine_02}}
     {targetBone {boneName=spine_03}}
     {targetBone {boneName=spine_04}}
     {targetBone {boneName=spine_05}}
     {targetBone {boneName=upperarm_twist_01_l}}
    ]
    constraintWeights [1,1,1,1,1]
   }
   n151 = "Constraint - dyn_jacket_c_constraint (PPPPPPPP)"!(
    ComponentPose=@n150
    Alpha=1
    ConstraintWeights_0=0.0166
    ConstraintWeights_1:float=0.0139
    ConstraintWeights_2:float=0.1631
    ConstraintWeights_3:float=0.3699
    ConstraintWeights_4:float=0.354
    ConstraintWeights_5:float=0.0728
    ConstraintWeights_6:float=0.009
    ConstraintWeights_7:float=0.0007
   ) -> (Pose) {
    alphaBoolBlend {blendOption=Linear}
    boneToModify {boneName=dyn_jacket_c_constraint}
    constraintSetup [
     {targetBone {boneName=clavicle_l}}
     {targetBone {boneName=spine_01}}
     {targetBone {boneName=spine_02}}
     {targetBone {boneName=spine_03}}
     {targetBone {boneName=spine_04}}
     {targetBone {boneName=spine_05}}
     {targetBone {boneName=upperarm_twist_01_l}}
     {targetBone {boneName=pelvis}}
    ]
    constraintWeights [1,1,1,1,1,1,1,1]
   }
   n152 = "Constraint - dyn_jacket_d_constraint (PPPPPPPP)"!(
    ComponentPose=@n151
    Alpha=1
    ConstraintWeights_0=0.0023
    ConstraintWeights_1:float=0.0011
    ConstraintWeights_2:float=0.0076
    ConstraintWeights_3:float=0.132
    ConstraintWeights_4:float=0.3317
    ConstraintWeights_5:float=0.3415
    ConstraintWeights_6:float=0.1804
    ConstraintWeights_7:float=0.0034
   ) -> (Pose) {
    alphaBoolBlend {blendOption=Linear}
    boneToModify {boneName=dyn_jacket_d_constraint}
    constraintSetup [
     {targetBone {boneName=clavicle_l}}
     {targetBone {boneName=upperarm_twist_01_r}}
     {targetBone {boneName=spine_01}}
     {targetBone {boneName=spine_02}}
     {targetBone {boneName=spine_03}}
     {targetBone {boneName=spine_04}}
     {targetBone {boneName=spine_05}}
     {targetBone {boneName=clavicle_r}}
    ]
    constraintWeights [1,1,1,1,1,1,1,1]
   }
   n153 = "Constraint - dyn_jacket_e_constraint (PPPPP)"!(
    ComponentPose=@n152
    Alpha=1
    ConstraintWeights_0=0.0292
    ConstraintWeights_1:float=0.0936
    ConstraintWeights_2:float=0.4027
    ConstraintWeights_3:float=0.3859
    ConstraintWeights_4:float=0.0886
   ) -> (Pose) {
    alphaBoolBlend {blendOption=Linear}
    boneToModify {boneName=dyn_jacket_e_constraint}
    constraintSetup [
     {targetBone {boneName=upperarm_twist_01_r}}
     {targetBone {boneName=spine_02}}
     {targetBone {boneName=spine_03}}
     {targetBone {boneName=spine_04}}
     {targetBone {boneName=spine_05}}
    ]
    constraintWeights [1,1,1,1,1]
   }
   n154 = "Constraint - dyn_jacket_f_constraint (PPPPP)"!(
    ComponentPose=@n153
    Alpha=1
    ConstraintWeights_0=0.0688
    ConstraintWeights_1:float=0.0891
    ConstraintWeights_2:float=0.3609
    ConstraintWeights_3:float=0.4586
    ConstraintWeights_4:float=0.0226
   ) -> (Pose) {
    alphaBoolBlend {blendOption=Linear}
    boneToModify {boneName=dyn_jacket_f_constraint}
    constraintSetup [
     {targetBone {boneName=upperarm_twist_01_r}}
     {targetBone {boneName=spine_02}}
     {targetBone {boneName=spine_03}}
     {targetBone {boneName=spine_04}}
     {targetBone {boneName=spine_05}}
    ]
    constraintWeights [1,1,1,1,1]
   }
   n155 = "Constraint - dyn_jacket_g_constraint (PPPPPPPP)"!(
    ComponentPose=@n154
    Alpha=1
    ConstraintWeights_0=0.0001
    ConstraintWeights_1:float=0.0066
    ConstraintWeights_2:float=0.0834
    ConstraintWeights_3:float=0.2143
    ConstraintWeights_4:float=0.4947
    ConstraintWeights_5:float=0.1957
    ConstraintWeights_6:float=0.0028
    ConstraintWeights_7:float=0.0024
   ) -> (Pose) {
    alphaBoolBlend {blendOption=Linear}
    boneToModify {boneName=dyn_jacket_g_constraint}
    constraintSetup [
     {targetBone {boneName=upperarm_twist_01_r}}
     {targetBone {boneName=spine_01}}
     {targetBone {boneName=spine_02}}
     {targetBone {boneName=spine_03}}
     {targetBone {boneName=spine_04}}
     {targetBone {boneName=spine_05}}
     {targetBone {boneName=clavicle_r}}
     {targetBone {boneName=pelvis}}
    ]
    constraintWeights [1,1,1,1,1,1,1,1]
   }
   Save cached pose 'prefrontend'(Pose=Component To Local(ComponentPose=@n155)) {cacheName=prefrontend}
   Reroute Node(InputPin:Vector=@n70) -> (OutputPin:Vector)
   Reroute Node(InputPin=@n31)
   Reroute Node(InputPin=@n50)
  ]}
 EventGraph {nodes [
   n1 = Event BlueprintUpdateAnimation()
   > n2 = Sequence!(execute) -> (
    then_0
    then_1
    then_2:Exec
    then_3:Exec
    then_4:Exec
    then_5:Exec
    then_6:Exec
    then_7:Exec
    then_8:Exec
    then_9:Exec
    then_10:Exec
    then_11:Exec
    then_12:Exec
    then_13:Exec
    then_14:Exec
    then_15:Exec
    then_16:Exec
   )
   > n3 = Gravity Override State Function DEPRECATED(CharacterPartAnimInstance=Self-Reference(), GameTimeInSeconds=GetGameTimeInSeconds(), GravityOverrideParams=Get Gravity Override Params())
   > Set GravityOverride(GravityOverride=@n3)
   n4 = Break Fort Anim Input Slope Info!(FortAnimInput_SlopeInfo:"Fort Anim Input Slope Info Structure (by ref)"=Get SlopeInfoInput() -> (SlopeInfoInput:Fort Anim Input Slope Info Structure)) -> (RootSlopeTranslation:Vector=~, LeftFootSlopeTranslation:Vector=~, RightFootSlopeTranslation:Vector=~, SlopeRotation:Rotator=~) {is=title K2Node_BreakStruct}
   Set slope_root_trans(execute=@n2.then_1, slope_root_trans:Vector=@n4) -> (Output_Get:Vector)
   > Set slope_foot_rot(slope_foot_rot:Rotator=@n4.SlopeRotation) -> (Output_Get:Rotator)
   n5 = Branch(execute=@n2.then_11, Condition=AND Boolean(A=Get bIsSkyDiving(), B=Get wearingBackpack() -> (wearingBackpack:bool)))
   > Set AccScale(AccScale:Vector="0.1,0.1,0.1") -> (Output_Get:Vector)
   > Set VelScale(VelScale:Vector="0.1,0.1,0.1") -> (Output_Get:Vector)
   Set AccScale(execute=@n5.else, AccScale="0.2,0.2,0.2")
   > Set VelScale(VelScale="0.2,0.2,0.2")
   n6 = Branch(execute=@n2.then_12, Condition=Get bIsPlayingEmote())
   > Set finger scale(finger scale:bool=false) -> (Output_Get:bool)
   Set finger scale(execute=@n6.else, finger scale=true)
   n7 = Player Movement Function DEPRECATED!(
    execute=@n2.then_13
    CharacterPartAnimInstance:Custom Character Part Anim Instance Object Reference=Self-Reference()
    X Dot Product:double=1
    Y Dot Product:double=1
    Z Dot Product:double=1
    DrawDebug:bool=false
    Debug Scalar:double=50
   ) -> (then, X Movement:double, Y Movement:double, Z Movement:double)
   n8 = "Absolute (Float)"(A=@n7)
   > Set BwdMovement(BwdMovement:double=float * float(
     A=MapRangeClamped(Value=@n8, InRangeA=0, InRangeB=1, OutRangeA=1, OutRangeB=0.4)
     B=MapRangeClamped(Value=@n7.Y Movement, InRangeA=0, InRangeB=-1, OutRangeA=1, OutRangeB=0)
    )) -> (Output_Get:double)
   n9 = Branch(execute=@n2.then_14, Condition=Get bIsSkyDiving())
   > n10 = Branch(Condition="float > float"(A=Get PawnVelocityZ(), B=50) {is=title K2Node_CallFunction})
   > Set BwdMovement(BwdMovement=float * float(A=Get BwdMovement(), B=0))
   Set BwdMovement(execute=[@n9.else, @n10.else], BwdMovement=float * float(A=Get BwdMovement(), B=1))
   n11 = GetPartSkeletalMeshComponent()
   Is Valid(exec=@n2.then_15, InputObject=@n11)
   > Delay(Duration=0)
   > SetVectorParameterValueOnMaterials!(
    execute
    self:Mesh Component Object Reference=@n11
    ParameterName:Name=Cube Location
    ParameterValue:Vector=TransformLocation!(T:"Transform (by ref)"=GetSocketTransform(self=@n11, InSocketName=spine_05, TransformSpace=RTS_Component), Location:Vector="0, 0, 0") -> (ReturnValue:Vector)
   ) -> (then)
   > SetVectorParameterValueOnMaterials(
    self=@n11
    ParameterName=Core Location R
    ParameterValue=TransformLocation(T=GetSocketTransform(self=@n11, InSocketName=thigh_twist_01_r, TransformSpace=RTS_Component), Location="0, 0, 0")
   )
   > SetVectorParameterValueOnMaterials(
    self=@n11
    ParameterName=Core Location L
    ParameterValue=TransformLocation(T=GetSocketTransform(self=@n11, InSocketName=thigh_twist_01_l, TransformSpace=RTS_Component), Location="0, 0, 0")
   )
   Reroute Node(InputPin:Exec=@n2.then_16) -> (OutputPin:Exec)
   n12 = ComplexSimSpaceFunction DEPRECATED!(
    execute=@n2.then_2
    CharacterPartAnimInstance:Custom Character Part Anim Instance Object Reference=Self-Reference()
    InterpSpeedMax:double=9
    PawnSpeedMax:double=20
    RInterpErrorTolerance:double=1
    DeltaTime:double=0
    DebugDraw:bool=false
    sim_space_rotation_in:Rotator=Get sim_space_rotation() -> (sim_space_rotation:Rotator)
    inComponentSpaceIn:bool=Get InComponentSpace() -> (InComponentSpace:bool)
    inComponentSpaceLastFrameIn:bool=Get InComponentSpaceLastFrame() -> (InComponentSpaceLastFrame:bool)
    lockedOnSimSpaceIn:bool=Get LockedOnSimSpace() -> (LockedOnSimSpace:bool)
    InterpSpeedIn:double=Get InterpSpeed() -> (InterpSpeed:double)
   ) -> (
    then
    sim_space_rotation:Rotator
    inComponentSpaceOut:bool
    inComponentSpaceLastFrameOut:bool
    lockedOnSimSpaceOut:bool
    InterpSpeedOut:double
   )
   > Set sim_space_rotation(sim_space_rotation:Rotator=@n12) -> (Output_Get:Rotator)
   > Set InComponentSpace(InComponentSpace:bool=@n12.inComponentSpaceOut) -> (Output_Get:bool)
   > Set InComponentSpaceLastFrame(InComponentSpaceLastFrame:bool=@n12.inComponentSpaceLastFrameOut) -> (Output_Get:bool)
   > Set LockedOnSimSpace(LockedOnSimSpace:bool=@n12.lockedOnSimSpaceOut) -> (Output_Get:bool)
   > Set InterpSpeed(InterpSpeed:double=@n12.InterpSpeedOut) -> (Output_Get:double)
   Branch(execute=@n2.then_3, Condition=OR Boolean(A=Get bIsSkyDiving(), B=Get bIsInFrontEnd()))
   > Set frontEndorSkydiving(frontEndorSkydiving:bool=true) -> (Output_Get:bool)
   n13 = GetOwnerSkeletalMeshComponent()
   n14 = BreakTransform(InTransform=GetDeltaTransformFromRefPose(self=@n13, BoneName=upperarm_r, BaseName=clavicle_r))
   Set LookUp(execute=@n2.then_4, LookUp:double=MapRangeClamped(Value=BreakRotator(InRot=@n14.Rotation).Yaw, InRangeA=90, InRangeB=120)) -> (Output_Get:double)
   n15 = BreakTransform(InTransform=GetDeltaTransformFromRefPose(self=@n13, BoneName=upperarm_r, BaseName=clavicle_r))
   > Set Lookup_Collar(Lookup_Collar:double=MapRangeClamped(Value=BreakRotator(InRot=@n15.Rotation).Yaw, InRangeA=40, InRangeB=90)) -> (Output_Get:double)
   n16 = BreakTransform(InTransform=GetDeltaTransformFromRefPose(self=@n13, BoneName=upperarm_l, BaseName=clavicle_l))
   > Set LookDown(LookDown:double=MapRangeClamped(Value=BreakRotator(InRot=@n16.Rotation).Yaw, InRangeA=50, InRangeB=15)) -> (Output_Get:double)
   n17 = BreakTransform(InTransform=GetDeltaTransformFromRefPose(self=@n13, BoneName=upperarm_r, BaseName=clavicle_r))
   > Set LookUpRifle(LookUpRifle:double=MapRangeClamped(Value=BreakRotator(InRot=@n17.Rotation).Yaw, InRangeA=60, InRangeB=78)) -> (Output_Get:double)
   Set jacket dynamics(execute=@n2.then_5, jacket dynamics:double=SelectFloat(A=0.5, B=1, bPickA=Get bIsCrouching())) -> (Output_Get:double)
   > Set Skydive_off(Skydive_off:bool=NOT Boolean(A=Get bIsSkyDiving())) -> (Output_Get:bool)
   n18 = Branch(execute=@n2.then_6, Condition=AND Boolean(A=Get bIsCrouching(), B=Get bIsAccelerating2D()))
   > Set crouch moving(crouch moving:bool=true) -> (Output_Get:bool)
   Set crouch moving(execute=@n18.else, crouch moving=false)
   n19 = Gravity Override State Function DEPRECATED(
    execute=@n2.then_7
    CharacterPartAnimInstance=Self-Reference()
    GameTimeInSeconds=GetGameTimeInSeconds()
    GravityOverrideParams=Get Gravity Override Params_0() -> (Gravity Override Params_0:Gravity Override Params Struct Structure)
   )
   > Set GravityOverride_2(GravityOverride_2:Vector=@n19) -> (Output_Get:Vector)
   n20 = CheckForBackbling()
   n21 = AND Boolean(A=@n20, B=NOT Boolean(A=@n20.bIsBackblingLong))
   Set wearingBackpack(execute=@n2.then_8, wearingBackpack:bool=@n21) -> (Output_Get:bool)
   > Set isWearingBackpack(isWearingBackpack:double="To Float (Boolean)"(InBool=@n21)) -> (Output_Get:double)
   > Set isNotWearingBackbling(isNotWearingBackbling:double="To Float (Boolean)"(InBool=NOT Boolean(A=@n20))) -> (Output_Get:double)
   > Set isWearingCape(isWearingCape:double="To Float (Boolean)"(InBool=@n20.bIsBackblingLong)) -> (Output_Get:double)
   > Set WearingCape(WearingCape:bool=AND Boolean(A=@n20, B=@n20.bIsBackblingLong)) -> (Output_Get:bool)
   n22 = Branch(execute=@n2.then_9, Condition=Get wearingBackpack())
   > Set backpack(backpack:double=0.5) -> (Output_Get:double)
   Set backpack(execute=@n22.else, backpack=1)
   n23 = Get Anim_Inst() -> (Anim_Inst:Anim Instance Object Reference)
   n24 = Montage_IsActive!(
    self:Anim Instance Object Reference=@n23
    Montage:Anim Montage Object Reference=/BRCosmetics/Animation/Game/MainPlayer/Emotes/CerealBox/Emote_CerealBox_CMM_M
   ) -> (ReturnValue:bool)
   Set EmoteAlpha(
    execute=@n2.then_10
    EmoteAlpha:double=FInterpTo(Current=Get EmoteAlpha(), Target="To Float (Boolean)"(InBool=@n24), DeltaTime=@n1.DeltaTimeX, InterpSpeed=0)
   ) -> (Output_Get:double)
   > Set emoteOverwriteJoint_Position(emoteOverwriteJoint_Position:double=Montage_GetPosition!(
     self:Anim Instance Object Reference=@n23
     Montage:Anim Montage Object Reference=&Emote_CerealBox_CMM_M
    ) -> (ReturnValue:float)) -> (Output_Get:double)
   Event OnCharacterCustomizationLoaded()
   > Delay(Duration=0.1)
   > ResetDynamics(InTeleportType=TeleportPhysics)
   > ResetDynamics(InTeleportType=ResetPhysics)
   Event OnEndCrouching()
   > Set crouch alpha(crouch alpha:double=0) -> (Output_Get:double)
   Event OnBeginCrouching()
   > Set crouch alpha(crouch alpha=1)
   Event BlueprintInitializeAnimation()
   n25 = GetOwnerSkeletalMeshComponent()
   > Branch(Condition=IsValid(Object=@n25))
   > n26 = Set Skel_Mesh(Skel_Mesh=@n25)
   > Set Anim_Inst(Anim_Inst:Anim Instance Object Reference=GetAnimInstance(self=@n26)) -> (Output_Get:Anim Instance Object Reference)
   Event OnEndSkydiving()
   > Delay(Duration=0.1)
   > ResetDynamics(InTeleportType=TeleportPhysics)
   > ResetDynamics(InTeleportType=ResetPhysics)
   n27 = Get Instigator!(self:Actor Object Reference=GetOwningActor()) -> (Instigator:Pawn Object Reference)
   Is Valid(InputObject=@n27)
   > n28 = Cast To FortPlayerPawn(Object=@n27)
   > n29 = Set As Fort Player Pawn(As Fort Player Pawn:Fort Player Pawn Object Reference=@n28) -> (Output_Get:Fort Player Pawn Object Reference)
   > Is Valid(InputObject=@n29)
   > Sequence!(execute) -> (then_0, then_1)
   n30 = GetDisplayName(Object=Get CurrentWeapon() -> (CurrentWeapon:Fort Weapon Object Reference))
   > n31 = Branch(Condition="Equal Exactly (String)"(A=@n30, B=B_LavaLeaf0))
   > Set pulsePistol_On(pulsePistol_On:bool=false) -> (Output_Get:bool)
   n32 = Branch(execute=@n31.else, Condition="Equal Exactly (String)"(A=@n30, B=B_LavaLeaf_1P0))
   > n33 = Set pulsePistol_On(pulsePistol_On=true)
   Do Once(execute=@n32.else, Reset=@n33.then)
   > RetriggerableDelay(Duration=10)
   > Set pulsePistol_On(pulsePistol_On=false)
  ]}
]
variables [
 AccScale {type=Vector}
 Anim_Inst {type=AnimInstance, default=null}
 As Fort Player Pawn {type=FortPlayerPawn}
 backpack {type=number}
 BwdMovement {type=number}
 crouch alpha {type=number}
 crouch moving {type=boolean}
 EmoteAlpha {type=number}
 emoteOverwriteJoint_Position {type=number}
 finger scale {type=boolean}
 frontEndorSkydiving {type=boolean}
 Gravity Override Params {
  type=GravityOverrideParamsStruct
  default {
   emoteOrMelee {gravityOverride=<0 0 -720>}
   front_End {gravityOverride=<0 0 -720>}
   onGround_Moving {gravityOverride=<0 0 -720>}
   onGround_Standing {gravityOverride=<0 0 -720>}
   skydive_FloatingDown {gravityOverride=<0 0 -720>}
   skydive_GoingUp {gravityOverride=<0 0 -200>}
   skydive_Parachuting {gravityOverride=<0 0 -720>, windAmplitude=<100 -250 200>, windFrequency=<0.75 3.5 4>, windOffset=<0 -300 200>}
  }
 }
 Gravity Override Params_0 {
  type=GravityOverrideParamsStruct
  default {
   skydive_DivingDown {gravityOverride=<0 0 0>}
   skydive_FloatingDown {gravityOverride=<0 0 360>}
   skydive_GoingUp {gravityOverride=<0 0 0>}
   skydive_Parachuting {gravityOverride=<0 0 -540>}
  }
 }
 GravityOverride {type=Vector}
 GravityOverride_2 {type=Vector}
 InComponentSpace {type=boolean}
 InComponentSpaceLastFrame {type=boolean}
 InterpSpeed {type=number}
 isNotWearingBackbling {type=number}
 isWearingBackpack {type=number}
 isWearingCape {type=number}
 jacket dynamics {type=number}
 LockedOnSimSpace {type=boolean}
 LookDown {type=number}
 LookUp {type=number}
 Lookup_Collar {type=number}
 LookUpRifle {type=number}
 pulsePistol_On {type=boolean}
 sim_space_rotation {type=Rotator}
 Skel_Mesh {type=SkeletalMeshComponent}
 Skydive_off {type=boolean}
 slope_foot_rot {type=Rotator}
 slope_root_trans {type=Vector}
 VelScale {type=Vector}
 wearingBackpack {type=boolean}
 WearingCape {type=boolean}
]
