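# Physics asset for the face-accessory part of the Cereal Box medium male character, frontend variant, covering a dynamic hair chain and upper body.
# Capsules cover spine, neck, clavicles and upper arms, linked by free angular joints up to the face-accessory and hair master bones.
# From the master, the hair chain runs 05, 01–04, then 06, mostly tapering boxes.
# Joints beyond 05 carry narrower angular limits, with tighter twist at the tip.
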PhysicsAsset /BRCosmetics/Characters/Player/Male/Medium/Bodies/M_MED_Cereal_Box/Meshes/Parts/M_MED_Cereal_Box_FaceAcc_Frontend_Physics
bodies {
 clavicle_l [{shapeType=Capsule, length=13.165, radius=6.8515, shapeName=clavicle_l_capsule, center=<6.932 -2.5 -1>, rotation=<rot 90 0.0099 0.0099>}]
 clavicle_r [{shapeType=Capsule, length=13.165, radius=6.8515, shapeName=clavicle_r_capsule, center=<-6.9321 2.5 1>, rotation=<rot 90 0 0>}]
 dyn_face_accessory_MSTR [{shapeType=Capsule, length=0.5, radius=0.5, shapeName=dyn_face_accessory_MSTR_capsule}]
 dyn_hair_01 [{shapeType=Box, extentX=14.6854, extentY=20.6366, extentZ=11.6105, shapeName=dyn_hair_01_box, center=<8.6921 0 0>, rotation=<rot -3.9415 0 0>}]
 dyn_hair_02 [{shapeType=Box, extentX=11.4658, extentY=19.4036, extentZ=9.8606, shapeName=dyn_hair_02_box, center=<6 0 0>, rotation=<rot 1.452 0 0>}]
 dyn_hair_03 [{shapeType=Box, extentX=9.0049, extentY=18.4973, extentZ=7.3561, shapeName=dyn_hair_03_box, center=<4.7967 0 0>, rotation=<rot 14.8812 0 0>}]
 dyn_hair_04 [{shapeType=Box, extentX=10.2877, extentY=13.8904, extentZ=3.5562, shapeName=dyn_hair_04_box, center=<5.1085 0 -1>, rotation=<rot -18.913 0 0>}]
 dyn_hair_05 [{shapeType=Capsule, length=0.505, radius=0.505, shapeName=dyn_hair_05_capsule, rotation=<rot 0 -90 0>}]
 dyn_hair_06 [{shapeType=Capsule, length=6.4262, radius=1.6383, shapeName=dyn_hair_06_capsule, center=<5 0 0>, rotation=<rot 0 0 90>}]
 dyn_hair_MSTR [{shapeType=Capsule, length=6.8838, radius=5.3676, shapeName=dyn_hair_MSTR_capsule, center=<2.437 0.0001 0>, rotation=<rot 0 0 90>}]
 neck_01 [{shapeType=Capsule, length=6.5159, radius=7.8791, shapeName=neck_01_capsule, center=<3.6721 -0.7603 0>, rotation=<rot 90 -1.023 10.424>}]
 neck_02 [{shapeType=Capsule, length=12.1983, radius=6.0378, shapeName=neck_02_capsule, center=<7.7804 -2.0179 0.0002>, rotation=<rot 90 0 -0.4675>}]
 spine_05 [{shapeType=Capsule, length=15.9714, radius=13.2886, shapeName=spine_05_capsule, center=<0 0.8077 0>}]
 upperarm_l [{shapeType=Capsule, length=16.2236, radius=7.5314, shapeName=upperarm_l_capsule, center=<6.932 0 0>, rotation=<rot 90 0 0>}]
 upperarm_r [{shapeType=Capsule, length=15.9186, radius=7.8364, shapeName=upperarm_r_capsule, center=<-6.932 0 0>, rotation=<rot 90 0 0>}]
}
constraints [
 {
  bone1Name=dyn_face_accessory_MSTR
  bone2Name=neck_02
  swing1LimitDegrees=45
  swing1Motion=Free
  swing2LimitDegrees=45
  swing2Motion=Free
  twistLimitDegrees=45
  twistMotion=Free
 }
 {
  bone1Name=neck_02
  bone2Name=neck_01
  swing1LimitDegrees=45
  swing1Motion=Free
  swing2LimitDegrees=45
  swing2Motion=Free
  twistLimitDegrees=45
  twistMotion=Free
 }
 {
  bone1Name=neck_01
  bone2Name=spine_05
  swing1LimitDegrees=45
  swing1Motion=Free
  swing2LimitDegrees=45
  swing2Motion=Free
  twistLimitDegrees=45
  twistMotion=Free
 }
 {
  bone1Name=clavicle_l
  bone2Name=spine_05
  swing1LimitDegrees=45
  swing1Motion=Free
  swing2LimitDegrees=45
  swing2Motion=Free
  twistLimitDegrees=45
  twistMotion=Free
 }
 {
  bone1Name=clavicle_r
  bone2Name=spine_05
  swing1LimitDegrees=45
  swing1Motion=Free
  swing2LimitDegrees=45
  swing2Motion=Free
  twistLimitDegrees=45
  twistMotion=Free
 }
 {
  bone1Name=upperarm_r
  bone2Name=clavicle_r
  swing1LimitDegrees=45
  swing1Motion=Free
  swing2LimitDegrees=45
  swing2Motion=Free
  twistLimitDegrees=45
  twistMotion=Free
 }
 {
  bone1Name=upperarm_l
  bone2Name=clavicle_l
  swing1LimitDegrees=45
  swing1Motion=Free
  swing2LimitDegrees=45
  swing2Motion=Free
  twistLimitDegrees=45
  twistMotion=Free
 }
 {
  bone1Name=dyn_hair_05
  bone2Name=dyn_hair_MSTR
  swing1LimitDegrees=45
  swing1Motion=Free
  swing2LimitDegrees=45
  swing2Motion=Free
  twistLimitDegrees=45
  twistMotion=Free
 }
 {
  bone1Name=dyn_hair_MSTR
  bone2Name=dyn_face_accessory_MSTR
  swing1LimitDegrees=45
  swing1Motion=Free
  swing2LimitDegrees=45
  swing2Motion=Free
  twistLimitDegrees=45
  twistMotion=Free
 }
 {bone1Name=dyn_hair_01, bone2Name=dyn_hair_05, swing1LimitDegrees=35, swing2LimitDegrees=35, twistLimitDegrees=35}
 {bone1Name=dyn_hair_02, bone2Name=dyn_hair_01, swing1LimitDegrees=35, swing2LimitDegrees=35, twistLimitDegrees=35}
 {bone1Name=dyn_hair_03, bone2Name=dyn_hair_02, swing1LimitDegrees=35, swing2LimitDegrees=35, twistLimitDegrees=35}
 {bone1Name=dyn_hair_04, bone2Name=dyn_hair_03, swing1LimitDegrees=35, swing2LimitDegrees=35, twistLimitDegrees=35}
 {bone1Name=dyn_hair_06, bone2Name=dyn_hair_04, swing1LimitDegrees=35, swing2LimitDegrees=35, twistLimitDegrees=15}
]
