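# Post-process style AnimBP for the Cereal Box male face-accessory part, adding secondary motion to its dyn_hair bones.
# The copied parent pose gets neck/head constraints and a skydive-driven look-at, then feeds RigidBody sims selected by front-end and backpack state.
# Weighted bone constraints and chained AnimDynamics springs finish the pose.
# The EventGraph updates backbling flags, state-based gravity and sim-space values, damps backpack skydiving motion, and resets dynamics after customization loads.
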AnimBlueprint /BRCosmetics/Characters/Player/Male/Medium/Bodies/M_MED_Cereal_Box/Meshes/Parts/M_MED_Cereal_Box_FaceAcc_AnimBP
target_skeleton=/Game/Characters/Player/Male/Medium/Bodies/M_MED_Cereal_Box/Meshes/Parts/M_MED_Cereal_Box_FaceAcc_Skeleton
graphs [
 AnimGraph {nodes [
   Anim Dynamics - Bone: dyn_hair_extra_g_01(Alpha=1, ExternalForce="0,0,0") {
    angularDampingOverride=0.9999
    angularSpringConstant=100
    bAngularSpring=true
    bGravityOverrideInSimSpace=true
    bOverrideAngularDamping=true
    simulationSpace=RootRelative
    alphaBoolBlend {blendOption=Linear}
    boundBone {boneName=dyn_hair_extra_g_01}
    componentAppliedLinearAccClamp=<3500 3500 3500>
    componentLinearAccScale=<1.2 1.2 1.2>
    componentLinearVelScale=<1.2 1.2 1.2>
    physicsBodyDefinitions [{
      sphereCollisionRadius=0
      boundBone {boneName=dyn_hair_extra_g_01}
      boxExtents=<20 10 10>
      constraintSetup {angularLimitsMax=<1 5 15>, angularLimitsMin=<-1 -30 -15>, angularTarget=<0 0 0>}
      localJointOffset=<2 0 0>
     }]
    relativeSpaceBone {boneName=dyn_root_simspace}
   }
   Copy Pose From Mesh() {bUseAttachedParent=true}
   Copy Pose From Mesh() {bUseAttachedParent=true}
   "Transform (Modify) Bone - Bone: dyn_hair_lookat"!(ComponentPose, bAlphaBoolEnabled:bool=True, Translation="0.000,20,0.000", Rotation="0,0,0", Scale="1,1,1") -> (Pose) {
    alphaInputType=Bool
    translationMode=BMM_Additive
    alphaBoolBlend {blendInTime=0.3, blendOption=Linear, blendOutTime=0.3}
    boneToModify {boneName=dyn_hair_lookat}
    translation {y=20}
   }
   n1 = Get GravityOverride()
   n2 = RigidBody!(ComponentPose=Local To Component(LocalPose=Use cached pose 'prerigids'()), Alpha, OverrideWorldGravity:Vector=@n1, ExternalForce) -> (Pose) {
    bClampLinearTranslationLimitToRefPose=true
    bOverrideWorldGravity=true
    overridePhysicsAsset=/Game/Characters/Player/Male/Medium/Bodies/M_MED_Cereal_Box/Meshes/Parts/M_MED_Cereal_Box_FaceAcc_Physics
    simulationSpace=BaseBoneSpace
    alphaBoolBlend {blendOption=Linear}
    baseBoneRef {boneName=spine_05}
    componentAppliedLinearAccClamp=<5000 5000 5000>
    componentLinearAccScale=<0.35 0.35 0.55>
    componentLinearVelScale=<0.2 0.2 0.35>
    overrideWorldGravity {z=-720}
   }
   n3 = RigidBody!(
    ComponentPose=Local To Component(LocalPose=Use cached pose 'prerigids'())
    Alpha
    ExternalForce
    ComponentLinearAccScale:Vector=Get AccScale() -> (AccScale:Vector)
    ComponentLinearVelScale:Vector=Get VelScale() -> (VelScale:Vector)
   ) -> (Pose) {
    bClampLinearTranslationLimitToRefPose=true
    bOverrideWorldGravity=true
    overridePhysicsAsset=&M_MED_Cereal_Box_FaceAcc_Physics
    simulationSpace=BaseBoneSpace
    alphaBoolBlend {blendOption=Linear}
    baseBoneRef {boneName=spine_05}
    componentAppliedLinearAccClamp=<3000 3000 3000>
    componentLinearAccScale=<0.2 0.2 0.2>
    componentLinearVelScale=<0.2 0.2 0.2>
    overrideWorldGravity {z=-720}
   }
   n4 = "Constraint - dyn_hair_extra_a (PPP)"!(ComponentPose=@n3, Alpha=1, ConstraintWeights_0=0.4498, ConstraintWeights_1:float=0.5502, ConstraintWeights_2:float=0) -> (Pose) {
    alphaBoolBlend {blendOption=Linear}
    boneToModify {boneName=dyn_hair_extra_a}
    constraintSetup [{targetBone {boneName=dyn_hair_01}}, {targetBone {boneName=dyn_hair_02}}, {targetBone {boneName=dyn_hair_03}}]
    constraintWeights [1,1,1]
   }
   n5 = "Transform (Modify) Bone - Bone: dyn_hair_01"(ComponentPose=@n4, Alpha=1, Translation="0,0,0", Rotation="0,0,0", Scale="0.9,1,0.8") {
    scaleMode=BMM_Additive
    scaleSpace=BCS_BoneSpace
    translationMode=BMM_Additive
    alphaBoolBlend {blendOption=Linear}
    boneToModify {boneName=dyn_hair_01}
    scale {x=0.9, z=0.8}
   }
   n6 = "Transform (Modify) Bone - Bone: dyn_hair_02"(ComponentPose=@n5, Alpha=1, Translation="0,0,0", Rotation="0,0,0", Scale="0.9,1,0.8") {
    scaleMode=BMM_Additive
    scaleSpace=BCS_BoneSpace
    alphaBoolBlend {blendOption=Linear}
    boneToModify {boneName=dyn_hair_02}
    scale {x=0.9, z=0.8}
   }
   n7 = "Transform (Modify) Bone - Bone: dyn_hair_03"(ComponentPose=@n6, Alpha=1, Translation="0,0,0", Rotation="0,0,0", Scale="0.9,1,0.8") {
    scaleMode=BMM_Additive
    scaleSpace=BCS_BoneSpace
    alphaBoolBlend {blendOption=Linear}
    boneToModify {boneName=dyn_hair_03}
    scale {x=0.9, z=0.8}
   }
   n8 = "Transform (Modify) Bone - Bone: dyn_hair_04"(ComponentPose=@n7, Alpha=1, Translation="0,0,0", Rotation="0,0,0", Scale="0.9,1,0.8") {
    scaleMode=BMM_Additive
    scaleSpace=BCS_BoneSpace
    alphaBoolBlend {blendOption=Linear}
    boneToModify {boneName=dyn_hair_04}
    scale {x=0.9, z=0.8}
   }
   n9 = Blend Poses by bool(
    BlendPose_0=Component To Local(ComponentPose=@n8)
    BlendPose_1=Component To Local(ComponentPose=@n2)
    bActiveValue=Get wearingBackpack() -> (wearingBackpack:bool)
   ) {blendType=Linear}
   n10 = RigidBody!(ComponentPose=Local To Component(LocalPose=Use cached pose 'prerigids'()), Alpha, OverrideWorldGravity:Vector=@n1, ExternalForce) -> (Pose) {
    bClampLinearTranslationLimitToRefPose=true
    bOverrideWorldGravity=true
    overridePhysicsAsset=&/M_MED_Cereal_Box_FaceAcc_Frontend_Physics
    simulationSpace=BaseBoneSpace
    alphaBoolBlend {blendOption=Linear}
    baseBoneRef {boneName=neck_01}
    componentAppliedLinearAccClamp=<2000 2000 2000>
    componentLinearAccScale=<0.3 0.3 0.5>
    componentLinearVelScale=<0.2 0.2 0.3>
    overrideWorldGravity {z=-720}
   }
   n11 = Component To Local(ComponentPose=@n10)
   n12 = Get bIsInFrontEnd()
   n13 = "Constraint - dyn_hair_extra_a (PPP)"(
    ComponentPose=Local To Component(LocalPose=Blend Poses by bool(BlendPose_0=@n11, BlendPose_1=@n9, bActiveValue=@n12) {blendType=Linear})
    Alpha=1
    ConstraintWeights_0=0.4498
    ConstraintWeights_1=0.5502
    ConstraintWeights_2=0
   ) {
    alphaBoolBlend {blendOption=Linear}
    boneToModify {boneName=dyn_hair_extra_a}
    constraintSetup [{targetBone {boneName=dyn_hair_01}}, {targetBone {boneName=dyn_hair_02}}, {targetBone {boneName=dyn_hair_03}}]
    constraintWeights [1,1,1]
   }
   n14 = "Constraint - dyn_hair_extra_b (PPPP)"!(
    ComponentPose=@n13
    Alpha=1
    ConstraintWeights_0=0.0031
    ConstraintWeights_1:float=0.0151
    ConstraintWeights_2:float=0.0002
    ConstraintWeights_3:float=0.9816
   ) -> (Pose) {
    alphaBoolBlend {blendOption=Linear}
    boneToModify {boneName=dyn_hair_extra_b}
    constraintSetup [
     {targetBone {boneName=dyn_hair_01}}
     {targetBone {boneName=dyn_hair_02}}
     {targetBone {boneName=dyn_face_accessory_MSTR}}
     {targetBone {boneName=dyn_hair_03}}
    ]
    constraintWeights [1,1,1,1]
   }
   n15 = "Constraint - dyn_hair_extra_c (PPPP)"!(
    ComponentPose=@n14
    Alpha=1
    ConstraintWeights_0=0.0487
    ConstraintWeights_1:float=0.4959
    ConstraintWeights_2:float=0
    ConstraintWeights_3:float=0.4555
   ) -> (Pose) {
    alphaBoolBlend {blendOption=Linear}
    boneToModify {boneName=dyn_hair_extra_c}
    constraintSetup [
     {targetBone {boneName=dyn_hair_01}}
     {targetBone {boneName=dyn_hair_02}}
     {targetBone {boneName=dyn_face_accessory_MSTR}}
     {targetBone {boneName=dyn_hair_03}}
    ]
    constraintWeights [1,1,1,1]
   }
   n16 = "Constraint - dyn_hair_extra_d (P)"!(ComponentPose=@n15, Alpha=1, ConstraintWeights_0=1, ConstraintWeights_1:float=0) -> (Pose) {
    alphaBoolBlend {blendOption=Linear}
    boneToModify {boneName=dyn_hair_extra_d}
    constraintSetup [{targetBone {boneName=dyn_hair_03}}]
    constraintWeights [1,0]
   }
   n17 = "Constraint - dyn_hair_extra_e (PPPP)"!(
    ComponentPose=@n16
    Alpha=1
    ConstraintWeights_0=0.608
    ConstraintWeights_1:float=0.3328
    ConstraintWeights_2:float=0.0338
    ConstraintWeights_3:float=0.0254
   ) -> (Pose) {
    alphaBoolBlend {blendOption=Linear}
    boneToModify {boneName=dyn_hair_extra_e}
    constraintSetup [
     {targetBone {boneName=dyn_hair_01}}
     {targetBone {boneName=dyn_hair_02}}
     {targetBone {boneName=dyn_face_accessory_MSTR}}
     {targetBone {boneName=dyn_hair_03}}
    ]
    constraintWeights [1,1,1,1]
   }
   n18 = "Constraint - dyn_hair_extra_f (P)"!(ComponentPose=@n17, Alpha=1, ConstraintWeights_0=1, ConstraintWeights_1:float=0) -> (Pose) {
    alphaBoolBlend {blendOption=Linear}
    boneToModify {boneName=dyn_hair_extra_f}
    constraintSetup [{targetBone {boneName=dyn_face_accessory_MSTR}}]
    constraintWeights [1,0]
   }
   n19 = "Constraint - dyn_hair_extra_h (PPP)"!(ComponentPose=@n18, Alpha=1, ConstraintWeights_0=0.3029, ConstraintWeights_1:float=0.3651, ConstraintWeights_2:float=0.332) -> (Pose) {
    alphaBoolBlend {blendOption=Linear}
    boneToModify {boneName=dyn_hair_extra_h}
    constraintSetup [{targetBone {boneName=dyn_hair_01}}, {targetBone {boneName=dyn_hair_02}}, {targetBone {boneName=dyn_hair_03}}]
    constraintWeights [1,1,1]
   }
   n20 = "Constraint - dyn_hair_extra_g_MSTR (P)"(ComponentPose=@n19, Alpha=1, ConstraintWeights_0=1) {
    alphaBoolBlend {blendOption=Linear}
    boneToModify {boneName=dyn_hair_extra_g_MSTR}
    constraintSetup [{targetBone {boneName=dyn_face_accessory_MSTR}}]
    constraintWeights [1]
   }
   n21 = "Constraint - dyn_hair_extra_g_01 (R)"(ComponentPose=@n20, Alpha=0.5, ConstraintWeights_0=1) {
    alphaBoolBlend {blendOption=Linear}
    boneToModify {boneName=dyn_hair_extra_g_01}
    constraintSetup [{transformType=Rotation, targetBone {boneName=dyn_hair_01}}]
    constraintWeights [1]
   }
   n22 = Anim Dynamics - Bone: dyn_hair_extra_g_02(ComponentPose=@n21, Alpha=1, ExternalForce="0,0,0") {
    angularDampingOverride=0.9999
    angularSpringConstant=25
    bAngularSpring=true
    bGravityOverrideInSimSpace=true
    bOverrideAngularDamping=true
    simulationSpace=RootRelative
    alphaBoolBlend {blendOption=Linear}
    boundBone {boneName=dyn_hair_extra_g_02}
    componentAppliedLinearAccClamp=<3500 3500 3500>
    componentLinearAccScale=<0.6 0.6 0.8>
    componentLinearVelScale=<0.6 0.6 0.8>
    physicsBodyDefinitions [{
      sphereCollisionRadius=0
      boundBone {boneName=dyn_hair_extra_g_02}
      boxExtents=<20 15 15>
      constraintSetup {angularLimitsMax=<1 -5 6>, angularLimitsMin=<-1 -45 1>, angularTarget=<0 0 0>}
      localJointOffset=<1 0 0>
     }]
    relativeSpaceBone {boneName=dyn_root_simspace}
   }
   n23 = Anim Dynamics - Bone: dyn_hair_extra_g_03(ComponentPose=@n22, Alpha=1, ExternalForce="0,0,0") {
    angularDampingOverride=0.9999
    angularSpringConstant=25
    bAngularSpring=true
    bGravityOverrideInSimSpace=true
    bOverrideAngularDamping=true
    simulationSpace=RootRelative
    alphaBoolBlend {blendOption=Linear}
    boundBone {boneName=dyn_hair_extra_g_03}
    componentAppliedLinearAccClamp=<3500 3500 3500>
    componentLinearAccScale=<0.6 0.6 0.8>
    componentLinearVelScale=<0.6 0.6 0.8>
    physicsBodyDefinitions [{
      sphereCollisionRadius=0
      boundBone {boneName=dyn_hair_extra_g_03}
      boxExtents=<20 15 15>
      constraintSetup {angularLimitsMax=<1 -5 8>, angularLimitsMin=<-1 -45 1>, angularTarget=<0 0 0>}
      localJointOffset=<1 0 0>
     }]
    relativeSpaceBone {boneName=dyn_root_simspace}
   }
   n24 = Anim Dynamics - Bone: dyn_hair_top_01(ComponentPose=@n23, Alpha=1, ExternalForce="0,0,0") {
    angularDampingOverride=0.9999
    angularSpringConstant=500
    bAngularSpring=true
    bGravityOverrideInSimSpace=true
    bOverrideAngularDamping=true
    simulationSpace=BoneRelative
    alphaBoolBlend {blendOption=Linear}
    boundBone {boneName=dyn_hair_top_01}
    componentAppliedLinearAccClamp=<3500 3500 3500>
    componentLinearAccScale=<1.2 1.2 1.2>
    componentLinearVelScale=<1.2 1.2 1.2>
    physicsBodyDefinitions [{
      sphereCollisionRadius=0
      boundBone {boneName=dyn_hair_top_01}
      boxExtents=<15 10 10>
      constraintSetup {angularLimitsMax=<1 5 15>, angularLimitsMin=<-1 -45 -15>, angularTarget=<0 0 0>}
      localJointOffset=<5 0 0>
     }]
    relativeSpaceBone {boneName=dyn_root_simspace}
   }
   n25 = Anim Dynamics - Bone: dyn_hair_top_02(ComponentPose=@n24, Alpha=1, ExternalForce="0,0,0") {
    angularDampingOverride=0.9999
    angularSpringConstant=500
    bAngularSpring=true
    bGravityOverrideInSimSpace=true
    bOverrideAngularDamping=true
    simulationSpace=BoneRelative
    alphaBoolBlend {blendOption=Linear}
    boundBone {boneName=dyn_hair_top_02}
    componentAppliedLinearAccClamp=<3500 3500 3500>
    componentLinearAccScale=<1.2 1.2 1.2>
    componentLinearVelScale=<1.2 1.2 1.2>
    physicsBodyDefinitions [{
      sphereCollisionRadius=0
      boundBone {boneName=dyn_hair_top_02}
      boxExtents=<15 10 10>
      constraintSetup {angularLimitsMax=<1 5 15>, angularLimitsMin=<-1 -45 -15>, angularTarget=<0 0 0>}
      localJointOffset=<5 0 0>
     }]
    relativeSpaceBone {boneName=dyn_root_simspace}
   }
   n26 = Anim Dynamics - Bone: dyn_hair_top_03(ComponentPose=@n25, Alpha=1, ExternalForce="0,0,0") {
    angularDampingOverride=0.9999
    angularSpringConstant=500
    bAngularSpring=true
    bGravityOverrideInSimSpace=true
    bOverrideAngularDamping=true
    simulationSpace=BoneRelative
    alphaBoolBlend {blendOption=Linear}
    boundBone {boneName=dyn_hair_top_03}
    componentAppliedLinearAccClamp=<3500 3500 3500>
    componentLinearAccScale=<1.2 1.2 1.2>
    componentLinearVelScale=<1.2 1.2 1.2>
    physicsBodyDefinitions [{
      sphereCollisionRadius=0
      boundBone {boneName=dyn_hair_top_03}
      boxExtents=<15 10 10>
      constraintSetup {angularLimitsMax=<1 5 15>, angularLimitsMin=<-1 -45 -15>, angularTarget=<0 0 0>}
      localJointOffset=<5 0 0>
     }]
    relativeSpaceBone {boneName=dyn_root_simspace}
   }
   n27 = Anim Dynamics - Bone: dyn_hair_top_04(ComponentPose=@n26, Alpha=1, ExternalForce="0,0,0") {
    angularDampingOverride=0.9999
    angularSpringConstant=500
    bAngularSpring=true
    bGravityOverrideInSimSpace=true
    bOverrideAngularDamping=true
    simulationSpace=BoneRelative
    alphaBoolBlend {blendOption=Linear}
    boundBone {boneName=dyn_hair_top_04}
    componentAppliedLinearAccClamp=<3500 3500 3500>
    componentLinearAccScale=<1.2 1.2 1.2>
    componentLinearVelScale=<1.2 1.2 1.2>
    physicsBodyDefinitions [{
      sphereCollisionRadius=0
      boundBone {boneName=dyn_hair_top_04}
      boxExtents=<15 10 10>
      constraintSetup {angularLimitsMax=<1 5 15>, angularLimitsMin=<-1 -45 -15>, angularTarget=<0 0 0>}
      localJointOffset=<5 0 0>
     }]
    relativeSpaceBone {boneName=dyn_root_simspace}
   }
   n28 = Anim Dynamics - Bone: dyn_hair_top_05(ComponentPose=@n27, Alpha=1, ExternalForce="0,0,0") {
    angularDampingOverride=0.9999
    angularSpringConstant=500
    bAngularSpring=true
    bGravityOverrideInSimSpace=true
    bOverrideAngularDamping=true
    simulationSpace=BoneRelative
    alphaBoolBlend {blendOption=Linear}
    boundBone {boneName=dyn_hair_top_05}
    componentAppliedLinearAccClamp=<3500 3500 3500>
    componentLinearAccScale=<1.2 1.2 1.2>
    componentLinearVelScale=<1.2 1.2 1.2>
    physicsBodyDefinitions [{
      sphereCollisionRadius=0
      boundBone {boneName=dyn_hair_top_05}
      boxExtents=<15 10 10>
      constraintSetup {angularLimitsMax=<1 5 15>, angularLimitsMin=<-1 -45 -15>, angularTarget=<0 0 0>}
      localJointOffset=<5 0 0>
     }]
    relativeSpaceBone {boneName=dyn_root_simspace}
   }
   n29 = Anim Dynamics - Bone: dyn_hair_top_06(ComponentPose=@n28, Alpha=1, ExternalForce="0,0,0") {
    angularDampingOverride=0.9999
    angularSpringConstant=500
    bAngularSpring=true
    bGravityOverrideInSimSpace=true
    bOverrideAngularDamping=true
    simulationSpace=BoneRelative
    alphaBoolBlend {blendOption=Linear}
    boundBone {boneName=dyn_hair_top_06}
    componentAppliedLinearAccClamp=<3500 3500 3500>
    componentLinearAccScale=<1.2 1.2 1.2>
    componentLinearVelScale=<1.2 1.2 1.2>
    physicsBodyDefinitions [{
      sphereCollisionRadius=0
      boundBone {boneName=dyn_hair_top_06}
      boxExtents=<15 10 10>
      constraintSetup {angularLimitsMax=<1 0 10>, angularLimitsMin=<-1 -35 -10>, angularTarget=<0 0 0>}
      localJointOffset=<5 0 0>
     }]
    relativeSpaceBone {boneName=dyn_root_simspace}
   }
   n30 = Anim Dynamics - Bone: dyn_hair_extra_a(ComponentPose=@n29, Alpha=1, ExternalForce="0,0,0") {
    angularDampingOverride=0.9999
    angularSpringConstant=100
    bAngularSpring=true
    bGravityOverrideInSimSpace=true
    bOverrideAngularDamping=true
    simulationSpace=RootRelative
    alphaBoolBlend {blendOption=Linear}
    boundBone {boneName=dyn_hair_extra_a}
    componentAppliedLinearAccClamp=<3500 3500 3500>
    componentLinearAccScale=<1.2 1.2 1.2>
    componentLinearVelScale=<1.2 1.2 1.2>
    physicsBodyDefinitions [{
      sphereCollisionRadius=0
      boundBone {boneName=dyn_hair_extra_a}
      boxExtents=<20 10 10>
      constraintSetup {angularLimitsMax=<1 10 15>, angularLimitsMin=<-1 -30 -15>, angularTarget=<0 0 0>}
      localJointOffset=<2 0 0>
     }]
    relativeSpaceBone {boneName=dyn_root_simspace}
   }
   n31 = Anim Dynamics - Bone: dyn_hair_extra_a1(ComponentPose=@n30, Alpha=1, ExternalForce="0,0,0") {
    angularDampingOverride=0.9999
    angularSpringConstant=100
    bAngularSpring=true
    bGravityOverrideInSimSpace=true
    bOverrideAngularDamping=true
    simulationSpace=RootRelative
    alphaBoolBlend {blendOption=Linear}
    boundBone {boneName=dyn_hair_extra_a1}
    componentAppliedLinearAccClamp=<3500 3500 3500>
    componentLinearAccScale=<1.2 1.2 1.2>
    componentLinearVelScale=<1.2 1.2 1.2>
    physicsBodyDefinitions [{
      sphereCollisionRadius=0
      boundBone {boneName=dyn_hair_extra_a1}
      boxExtents=<20 10 10>
      constraintSetup {angularLimitsMax=<1 10 15>, angularLimitsMin=<-1 -30 -15>, angularTarget=<0 0 0>}
      localJointOffset=<2 0 0>
     }]
    relativeSpaceBone {boneName=dyn_root_simspace}
   }
   n32 = Anim Dynamics - Bone: dyn_hair_extra_b(ComponentPose=@n31, Alpha=1, ExternalForce="0,0,0") {
    angularDampingOverride=0.9999
    angularSpringConstant=100
    bAngularSpring=true
    bGravityOverrideInSimSpace=true
    bOverrideAngularDamping=true
    simulationSpace=RootRelative
    alphaBoolBlend {blendOption=Linear}
    boundBone {boneName=dyn_hair_extra_b}
    componentAppliedLinearAccClamp=<3500 3500 3500>
    componentLinearAccScale=<1.2 1.2 1.2>
    componentLinearVelScale=<1.2 1.2 1.2>
    physicsBodyDefinitions [{
      sphereCollisionRadius=0
      boundBone {boneName=dyn_hair_extra_b}
      boxExtents=<20 10 10>
      constraintSetup {angularLimitsMax=<1 5 15>, angularLimitsMin=<-1 -30 -15>, angularTarget=<0 0 0>}
      localJointOffset=<2 0 0>
     }]
    relativeSpaceBone {boneName=dyn_root_simspace}
   }
   n33 = Anim Dynamics - Bone: dyn_hair_extra_c(ComponentPose=@n32, Alpha=1, ExternalForce="0,0,0") {
    angularDampingOverride=0.9999
    angularSpringConstant=100
    bAngularSpring=true
    bGravityOverrideInSimSpace=true
    bOverrideAngularDamping=true
    simulationSpace=RootRelative
    alphaBoolBlend {blendOption=Linear}
    boundBone {boneName=dyn_hair_extra_c}
    componentAppliedLinearAccClamp=<3500 3500 3500>
    componentLinearAccScale=<1.2 1.2 1.2>
    componentLinearVelScale=<1.2 1.2 1.2>
    physicsBodyDefinitions [{
      sphereCollisionRadius=0
      boundBone {boneName=dyn_hair_extra_c}
      boxExtents=<20 10 10>
      constraintSetup {angularLimitsMax=<1 5 15>, angularLimitsMin=<-1 -30 -15>, angularTarget=<0 0 0>}
      localJointOffset=<2 0 0>
     }]
    relativeSpaceBone {boneName=dyn_root_simspace}
   }
   n34 = Anim Dynamics - Bone: dyn_hair_extra_d(ComponentPose=@n33, Alpha=1, ExternalForce="0,0,0") {
    angularDampingOverride=0.9999
    angularSpringConstant=100
    bAngularSpring=true
    bGravityOverrideInSimSpace=true
    bOverrideAngularDamping=true
    simulationSpace=RootRelative
    alphaBoolBlend {blendOption=Linear}
    boundBone {boneName=dyn_hair_extra_d}
    componentAppliedLinearAccClamp=<3500 3500 3500>
    componentLinearAccScale=<1.2 1.2 1.2>
    componentLinearVelScale=<1.2 1.2 1.2>
    physicsBodyDefinitions [{
      sphereCollisionRadius=0
      boundBone {boneName=dyn_hair_extra_d}
      boxExtents=<20 10 10>
      constraintSetup {angularLimitsMax=<1 5 15>, angularLimitsMin=<-1 -30 -15>, angularTarget=<0 0 0>}
      localJointOffset=<2 0 0>
     }]
    relativeSpaceBone {boneName=dyn_root_simspace}
   }
   n35 = Anim Dynamics - Bone: dyn_hair_extra_e(ComponentPose=@n34, Alpha=1, ExternalForce="0,0,0") {
    angularDampingOverride=0.9999
    angularSpringConstant=100
    bAngularSpring=true
    bGravityOverrideInSimSpace=true
    bOverrideAngularDamping=true
    simulationSpace=RootRelative
    alphaBoolBlend {blendOption=Linear}
    boundBone {boneName=dyn_hair_extra_e}
    componentAppliedLinearAccClamp=<3500 3500 3500>
    componentLinearAccScale=<1.2 1.2 1.2>
    componentLinearVelScale=<1.2 1.2 1.2>
    physicsBodyDefinitions [{
      sphereCollisionRadius=0
      boundBone {boneName=dyn_hair_extra_e}
      boxExtents=<20 10 10>
      constraintSetup {angularLimitsMax=<1 5 15>, angularLimitsMin=<-1 -30 -15>, angularTarget=<0 0 0>}
      localJointOffset=<2 0 0>
     }]
    relativeSpaceBone {boneName=dyn_root_simspace}
   }
   n36 = Anim Dynamics - Bone: dyn_hair_extra_f(ComponentPose=@n35, Alpha=1, ExternalForce="0,0,0") {
    angularDampingOverride=0.9999
    angularSpringConstant=100
    bAngularSpring=true
    bGravityOverrideInSimSpace=true
    bOverrideAngularDamping=true
    simulationSpace=RootRelative
    alphaBoolBlend {blendOption=Linear}
    boundBone {boneName=dyn_hair_extra_f}
    componentAppliedLinearAccClamp=<3500 3500 3500>
    componentLinearAccScale=<1.2 1.2 1.2>
    componentLinearVelScale=<1.2 1.2 1.2>
    physicsBodyDefinitions [{
      sphereCollisionRadius=0
      boundBone {boneName=dyn_hair_extra_f}
      boxExtents=<20 10 10>
      constraintSetup {angularLimitsMax=<1 5 15>, angularLimitsMin=<-1 -30 -15>, angularTarget=<0 0 0>}
      localJointOffset=<2 0 0>
     }]
    relativeSpaceBone {boneName=dyn_root_simspace}
   }
   n37 = Anim Dynamics - Bone: dyn_hair_extra_h(ComponentPose=@n36, Alpha=1, ExternalForce="0,0,0") {
    angularDampingOverride=0.9999
    angularSpringConstant=100
    bAngularSpring=true
    bGravityOverrideInSimSpace=true
    bOverrideAngularDamping=true
    simulationSpace=RootRelative
    alphaBoolBlend {blendOption=Linear}
    boundBone {boneName=dyn_hair_extra_h}
    componentAppliedLinearAccClamp=<3500 3500 3500>
    componentLinearAccScale=<1.2 1.2 1.2>
    componentLinearVelScale=<1.2 1.2 1.2>
    physicsBodyDefinitions [{
      sphereCollisionRadius=0
      boundBone {boneName=dyn_hair_extra_h}
      boxExtents=<20 10 10>
      constraintSetup {angularLimitsMax=<1 5 15>, angularLimitsMin=<-1 -30 -15>, angularTarget=<0 0 0>}
      localJointOffset=<2 0 0>
     }]
    relativeSpaceBone {boneName=dyn_root_simspace}
   }
   Output Pose(Result=Component To Local(ComponentPose=@n37)) {layerGroup=None}
   n38 = "Constraint - dyn_neck_02 (R)"(ComponentPose=Local To Component(LocalPose=Copy Pose From Mesh() {bUseAttachedParent=true}), Alpha=1, ConstraintWeights_0=1) {
    alphaBoolBlend {blendOption=Linear}
    boneToModify {boneName=dyn_neck_02}
    constraintSetup [{transformType=Rotation, targetBone {boneName=neck_02}}]
    constraintWeights [1]
   }
   n39 = "Constraint - dyn_head (R)"(ComponentPose=@n38, Alpha=1, ConstraintWeights_0=1) {
    alphaBoolBlend {blendOption=Linear}
    boneToModify {boneName=dyn_head}
    constraintSetup [{transformType=Rotation, targetBone {boneName=head}}]
    constraintWeights [1]
   }
   n40 = "Transform (Modify) Bone - Bone: dyn_head"!(ComponentPose=@n39, bAlphaBoolEnabled:bool=True, Translation="0,0,0", Rotation="0,-5,-5", Scale="1,1,1") -> (Pose) {
    alphaInputType=Bool
    rotationMode=BMM_Additive
    scaleMode=BMM_Additive
    alphaBoolBlend {blendInTime=0.1, blendOption=Linear, blendOutTime=0.1}
    boneToModify {boneName=dyn_head}
    rotation {roll=-5, yaw=-5}
   }
   n41 = "Transform (Modify) Bone - Bone: dyn_hair_lookat"(ComponentPose=@n40, bAlphaBoolEnabled=Get bIsSkyDiving(), Translation="0.000,-60.765,133.430", Rotation="0,0,0", Scale="1,1,1") {
    alphaInputType=Bool
    translationMode=BMM_Additive
    alphaBoolBlend {blendInTime=0.3, blendOption=Linear, blendOutTime=0.3}
    boneToModify {boneName=dyn_hair_lookat}
    translation {y=-60.765, z=133.43}
   }
   n42 = Look At - Bone: dyn_hair_05(ComponentPose=@n41, Alpha=0.5) {
    alphaBoolBlend {blendOption=Linear}
    boneToModify {boneName=dyn_hair_05}
    lookAtLocation {x=0}
    lookAtTarget {boneReference {boneName=dyn_hair_lookat}}
    lookAt_Axis {axis=<1 0>}
   }
   Save cached pose 'prerigids'(Pose=Component To Local(ComponentPose=@n42)) {cacheName=prerigids}
   Get Skel_Mesh()
  ]}
 EventGraph {nodes [
   Event BlueprintUpdateAnimation()
   > n1 = Sequence!(execute) -> (then_0, then_1, then_2:Exec, then_3:Exec, then_4:Exec)
   n2 = CheckForBackbling()
   n3 = AND Boolean(A=@n2, B=NOT Boolean(A=@n2.bIsBackblingLong))
   > Set wearingBackpack(wearingBackpack:bool=@n3) -> (Output_Get:bool)
   > Set isWearingBackpack(isWearingBackpack:double="To Float (Boolean)"(InBool=@n3)) -> (Output_Get:double)
   > Set isNotWearingBackbling(isNotWearingBackbling:double="To Float (Boolean)"(InBool=NOT Boolean(A=@n2))) -> (Output_Get:double)
   > Set isWearingCape(isWearingCape:double="To Float (Boolean)"(InBool=@n2.bIsBackblingLong)) -> (Output_Get:double)
   > Set WearingCape(WearingCape:bool=AND Boolean(A=@n2, B=@n2.bIsBackblingLong)) -> (Output_Get:bool)
   n4 = Branch(execute=@n1.then_1, Condition=Get wearingBackpack())
   > Set backpack(backpack:double=0.5) -> (Output_Get:double)
   Set backpack(execute=@n4.else, backpack=1)
   n5 = Gravity Override State Function DEPRECATED(
    execute=@n1.then_2
    CharacterPartAnimInstance=Self-Reference()
    GameTimeInSeconds=GetGameTimeInSeconds()
    GravityOverrideParams=Get Gravity Override Params()
   )
   > Set GravityOverride(GravityOverride=@n5)
   n6 = ComplexSimSpaceFunction DEPRECATED!(
    execute=@n1.then_3
    CharacterPartAnimInstance:Custom Character Part Anim Instance Object Reference=Self-Reference()
    InterpSpeedMax:double=9
    PawnSpeedMax:double=20
    RInterpErrorTolerance:double=1
    DeltaTime:double=0
    DebugDraw:bool=false
    sim_space_rotation_in:Rotator=Get sim_space_rotation() -> (sim_space_rotation:Rotator)
    inComponentSpaceIn:bool=Get InComponentSpace() -> (InComponentSpace:bool)
    inComponentSpaceLastFrameIn:bool=Get InComponentSpaceLastFrame() -> (InComponentSpaceLastFrame:bool)
    lockedOnSimSpaceIn:bool=Get LockedOnSimSpace() -> (LockedOnSimSpace:bool)
    InterpSpeedIn:double=Get InterpSpeed() -> (InterpSpeed:double)
   ) -> (
    then
    sim_space_rotation:Rotator
    inComponentSpaceOut:bool
    inComponentSpaceLastFrameOut:bool
    lockedOnSimSpaceOut:bool
    InterpSpeedOut:double
   )
   > Set sim_space_rotation(sim_space_rotation:Rotator=@n6) -> (Output_Get:Rotator)
   > Set InComponentSpace(InComponentSpace:bool=@n6.inComponentSpaceOut) -> (Output_Get:bool)
   > Set InComponentSpaceLastFrame(InComponentSpaceLastFrame:bool=@n6.inComponentSpaceLastFrameOut) -> (Output_Get:bool)
   > Set LockedOnSimSpace(LockedOnSimSpace:bool=@n6.lockedOnSimSpaceOut) -> (Output_Get:bool)
   > Set InterpSpeed(InterpSpeed:double=@n6.InterpSpeedOut) -> (Output_Get:double)
   n7 = Branch(execute=@n1.then_4, Condition=AND Boolean(A=Get bIsSkyDiving(), B=Get wearingBackpack()))
   > Set backpack skydive(backpack skydive:bool=true) -> (Output_Get:bool)
   > Set AccScale(AccScale:Vector="0.1,0.1,0.1") -> (Output_Get:Vector)
   > Set VelScale(VelScale:Vector="0.1,0.1,0.1") -> (Output_Get:Vector)
   Set backpack skydive(execute=@n7.else, backpack skydive=false)
   > Set AccScale(AccScale="0.2,0.2,0.2")
   > Set VelScale(VelScale="0.2,0.2,0.2")
   Event OnCharacterCustomizationLoaded()
   > Delay(Duration=0.1)
   > ResetDynamics(InTeleportType=TeleportPhysics)
   > ResetDynamics(InTeleportType=ResetPhysics)
   Event BlueprintInitializeAnimation()
   > Set Skel_Mesh(Skel_Mesh=GetOwnerSkeletalMeshComponent())
  ]}
]
variables [
 AccScale {type=Vector}
 backpack {type=number}
 backpack skydive {type=boolean}
 Gravity Override Params {
  type=GravityOverrideParamsStruct
  default {
   emoteOrMelee {gravityOverride=<0 0 -720>}
   front_End {gravityOverride=<0 0 -720>}
   onGround_Moving {gravityOverride=<0 0 -720>}
   onGround_Standing {gravityOverride=<0 0 -720>}
   skydive_DivingDown {gravityOverride=<0 0 0>}
   skydive_FloatingDown {gravityOverride=<0 0 -100>, windOffset=<0 -300 200>}
   skydive_GoingUp {gravityOverride=<0 0 -100>}
   skydive_Parachuting {gravityOverride=<0 0 -240>}
   swimming {gravityOverride=<0 0 0>}
  }
 }
 GravityOverride {type=Vector}
 InComponentSpace {type=boolean}
 InComponentSpaceLastFrame {type=boolean}
 InterpSpeed {type=number}
 isNotWearingBackbling {type=number}
 isWearingBackpack {type=number}
 isWearingCape {type=number}
 LockedOnSimSpace {type=boolean}
 sim_space_rotation {type=Rotator}
 Skel_Mesh {type=SkeletalMeshComponent}
 VelScale {type=Vector}
 wearingBackpack {type=boolean}
 WearingCape {type=boolean}
]
